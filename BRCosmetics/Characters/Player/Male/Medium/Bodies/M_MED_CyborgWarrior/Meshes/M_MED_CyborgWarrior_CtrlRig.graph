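# Control Rig for the M_MED_CyborgWarrior player body cosmetic, applying corrective deformation to body and clothing bones.
# A sequence runs shared DeformRig, derives CrouchingButUp from pelvis height, then adjusts skirt, daggers, shirt, shorts and collar.
# Spherical pose readers, calf distances and deformRig attributes weight lower-garment offsets.
# Collar offsets blend from swing/twist head rotation (DeltaRotation) and head pose readers gated by Crouching and Targeting, some smoothed with AlphaInterp.
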ControlRigBlueprint /BRCosmetics/Characters/Player/Male/Medium/Bodies/M_MED_CyborgWarrior/Meshes/M_MED_CyborgWarrior_CtrlRig
graphs [
 Multiply_1_ContainedGraph {nodes [
   n1 = RigVMFunctionEntryNode()
   Multiply = &RigVMFunction_MathDoubleMul(A=@n1.A, B=@n1.B)
   RigVMFunctionReturnNode(Result=&RigVMFunction_MathDoubleMul(B=@n1.D, A=&RigVMFunction_MathDoubleMul(B=@n1.C, A=@Multiply.Result).Result).Result)
  ]}
 Multiply_8_ContainedGraph {nodes [
   n1 = RigVMFunctionEntryNode()
   RigVMFunctionReturnNode(Result=&RigVMFunction_MathDoubleMul(B=@n1.C, A=&RigVMFunction_MathDoubleMul(A=@n1.A, B=@n1.B).Result).Result)
  ]}
 Multiply_9_ContainedGraph {nodes [
   n1 = RigVMFunctionEntryNode()
   RigVMFunctionReturnNode(Result=&RigVMFunction_MathDoubleMul(B=@n1.C, A=&RigVMFunction_MathDoubleMul(A=@n1.A, B=@n1.B).Result).Result)
  ]}
 New Function_ContainedGraph {nodes [
   n1 = RigVMFunctionEntryNode()
   RigVMFunction_MathBoolToInteger = &RigVMFunction_MathBoolToInteger(Value=@n1.onlyTwist)
   RigVMFunction_MathQuaternionSwingTwist = &RigVMFunction_MathQuaternionSwingTwist(
    twistAxis=@n1.twistAxis
    Input=&RigUnit_GetRelativeTransformForItem(
     bChildInitial=False
     bParentInitial=False
     child=(Type=Bone,Name="None")
     parent=(Type=Bone,Name="None")
     child.Name=@n1.child
     parent.Name=@n1.parent
     parent.Type=@n1.parentType
    ).RelativeTransform.Rotation
   )
   RigVMFunction_MathQuaternionToEuler = &RigVMFunction_MathQuaternionToEuler(
    Value=RigVMDispatchNode(
     Values=((),())
     Index=@RigVMFunction_MathBoolToInteger.Result
     Values.0=@RigVMFunction_MathQuaternionSwingTwist.Swing
     Values.1=@RigVMFunction_MathQuaternionSwingTwist.Twist
    ) {template="DISPATCH_RigVMDispatch_SelectInt32(in Index,in Values,out Result)"}.Result
    RotationOrder=@n1.Rotation Order
   )
   > DISPATCH_RigVMDispatch_Print = RigVMDispatchNode(ScreenColor=(R=1,G=1,B=1,A=1), ScreenDuration=1.5, Enabled=@n1.debugRotation, Value=@RigVMFunction_MathQuaternionToEuler.Result) {template="DISPATCH_RigVMDispatch_Print(in Prefix,in Value,in Enabled,in ScreenDuration,in ScreenColor)"}
   Clamp = &RigVMFunction_MathIntClamp(Value=@n1.primAxis)
   DISPATCH_RigVMDispatch_SelectInt32 = RigVMDispatchNode(
    Values=(0,0,0)
    Index=@Clamp.Result
    Values.0=@RigVMFunction_MathQuaternionToEuler.Result.X
    Values.1=@RigVMFunction_MathQuaternionToEuler.Result.Y
    Values.2=@RigVMFunction_MathQuaternionToEuler.Result.Z
   ) {template="DISPATCH_RigVMDispatch_SelectInt32(in Index,in Values,out Result)"}
   Remap = &RigVMFunction_MathDoubleRemap(
    TargetMaximum=1
    TargetMinimum=0
    bClamp=False
    Value=@DISPATCH_RigVMDispatch_SelectInt32.Result
    SourceMaximum=@n1.max
    SourceMinimum=@n1.min
   )
   Multiply = &RigVMFunction_MathDoubleMul(A=&RigVMFunction_MathDoubleClamp(Value=@Remap.Result).Result, B=@n1.premult)
   > RigVMFunctionReturnNode(outputWeight=@Multiply.Result)
  ]}
 New Function_ContainedGraph {nodes [
   n1 = RigVMFunctionEntryNode()
   RigVMFunction_MathQuaternionToEuler = &RigVMFunction_MathQuaternionToEuler(
    RotationOrder=YXZ
    Value=&RigUnit_GetRelativeTransformForItem(
     Child=(Type=Bone,Name="head")
     Parent=(Type=Bone,Name="neck_01")
     bChildInitial=False
     bParentInitial=False
     Child.Name=@n1.Child
     Parent.Name=@n1.Parent
    ).RelativeTransform.Rotation
   )
   Remap = &RigVMFunction_MathDoubleRemap(
    TargetMaximum=1
    TargetMinimum=0
    SourceMaximum=@n1.Source Max
    SourceMinimum=@n1.Source Min
    Value=@RigVMFunction_MathQuaternionToEuler.Result.Z
   )
   > RigVMFunctionReturnNode(
    Delta_z=@Remap.Result
    Delta_X_Negative=&RigVMFunction_MathDoubleRemap(SourceMaximum=-60, SourceMinimum=0, TargetMaximum=1, TargetMinimum=0, Value=@RigVMFunction_MathQuaternionToEuler.Result.X).Result
    Delta_X_Positive=&RigVMFunction_MathDoubleRemap(SourceMaximum=60, SourceMinimum=0, TargetMaximum=1, TargetMinimum=0, Value=@RigVMFunction_MathQuaternionToEuler.Result.X).Result
   )
  ]}
 RigVMFunctionLibrary {nodes [
   DeltaRotation = RigVMCollapseNode(ExecuteContext=(), Rotation Order=ZYX, premult=1, twistAxis=(X=1,Y=0,Z=0))
   DeltaRotation_3 = RigVMCollapseNode(ExecuteContext=())
  ]}
 RigVMFunction_Sequence_1_ContainedGraph {nodes [
   n1 = RigVMFunctionEntryNode()
   RigVMFunction_Sequence = &RigVMFunction_Sequence(ExecuteContext=@n1.ExecuteContext)
   n2 = &RigVMFunction_Sequence(ExecuteContext=@RigVMFunction_Sequence.B)
   n3 = &RigVMFunction_Sequence(ExecuteContext=@n2.B)
   n4 = &RigVMFunction_Sequence(ExecuteContext=@n3.B)
   n5 = &RigVMFunction_Sequence(ExecuteContext=@n4.B)
   n6 = &RigVMFunction_Sequence(ExecuteContext=@n5.B)
   RigVMFunctionReturnNode(A=@RigVMFunction_Sequence.A, B=@n2.A, C=@n3.A, D=@n4.A, E=@n5.A, F=@n6.A, G=@n6.B)
  ]}
 RigVMModel {nodes [
   RigUnit_BeginExecution = &RigUnit_BeginExecution()
   RigVMFunction_Sequence = RigVMAggregateNode(ExecuteContext=@RigUnit_BeginExecution.ExecutePin)
   FN_DeformRig_FNC = RigVMFunctionReferenceNode(
    DeformRig_ Curve="(deformRig_l_upperarm_up_alpha_72_#1=False,deformRig_r_upperarm_up_alpha_73_#2=False,deformRig_l_upperarm_fr_alpha_74_#3=False,deformRig_r_upperarm_fr_alpha_75_CB51B8 ...(truncated)"
    DeformRig_Bool="(upperarm_1_#4=True,elbow_3_#5=True,wrist_7_#6=True,groin_8_#7=True,glute_17_#8=True,knee_9_40C408324CA261 ...(truncated)"
    DeformRig_Debug=(upperarm_debug_18_#4=False,wrist_debug_20_#6=False,lat_debug_26_#9=False,groinGlute_debug_27_#7=False)
    DeformRig_Value="(upperarm_up_18_#10=(Rotation=(X=0.0,Y=-0.0,Z=0.0,W=1.0),Translation=(X=0.0,Y=0.0,Z=8.0),Scale3D=(X=0.70,Y=1.0,Z=1.0)),upperarm_fwd_fr_19_#11=(Rotation=(X=0.0,Y=-0.0,Z=0.0,W=1.0),Translation=(X=0.0, ...(truncated)"
    ExecuteContext=@RigVMFunction_Sequence.B
   )
   RigUnit_GetRelativeTransformForItem = &RigUnit_GetRelativeTransformForItem(Child=(Type=Bone,Name="pelvis"), Parent=(Type=Bone,Name="root"), bChildInitial=False, bParentInitial=False)
   Remap = &RigVMFunction_MathDoubleRemap(
    SourceMaximum=30
    SourceMinimum=60
    TargetMaximum=0
    TargetMinimum=1
    bClamp=False
    Value=@RigUnit_GetRelativeTransformForItem.RelativeTransform.Translation.Z
   )
   n1 = &RigVMFunction_AnimEvalRichCurve(
    Curve="(EditorCurveData=(Keys=((InterpMode=RCIM_Cubic,TangentMode=RCTM_Break,TangentWeightMode=RCTWM_WeightedNone,Time=0,Value=0,ArriveTangent=0,ArriveTangentWeight=0,LeaveTangent=0,LeaveTangentWeight=0),(InterpMode=RCIM_Cubic,TangentMode=RCTM_Break,TangentWeight ...(truncated)"
    SourceMinimum=0
    TargetMinimum=0
    Value=@Remap.Result
   )
   RigVMVariableNode(Value=@n1.Result, ExecuteContext=@RigVMFunction_Sequence.C) {variable=CrouchingButUp}
   n2 = /Script/RigVM.RigVMFunction_MathBoolMake(Value=false) {is=struct RigVMUnitNode}
   n3 = &RigUnit_SphericalPoseReader(
    ActiveRegionScaleFactors=(PositiveWidth=1,NegativeWidth=1,PositiveHeight=1,NegativeHeight=1)
    ActiveRegionSize=0.35
    Debug=(bDrawDebug=true,bDraw2D=False,bDrawLocalAxes=False,DebugScale=25,DebugSegments=20,DebugThickness=0.25)
    DriverAxis=(X=-1,Y=0,Z=0)
    DriverItem=(Type=Bone,Name="thigh_l")
    FalloffRegionScaleFactors=(PositiveWidth=1,NegativeWidth=1,PositiveHeight=1,NegativeHeight=1)
    FalloffSize=0.55
    FlipHeightScaling=False
    FlipWidthScaling=False
    OptionalParentItem=(Type=Bone,Name="None")
    RotationOffset=(X=90,Y=0,Z=-20)
    Debug.bDrawDebug=@n2.Value
    ExecutePin=@RigVMFunction_Sequence.E
   )
   RigUnit_GetTransform = &RigUnit_GetTransform(Item=(Type=Bone,Name="base_shirt_front_l"), bInitial=False)
   n4 = &RigVMFunction_MathVectorDistance(
    A=@RigUnit_GetTransform.Transform.Translation
    B=&RigUnit_GetTransform(Item=(Type=Bone,Name="calf_l"), bInitial=False).Transform.Translation
   )
   > &RigUnit_ModifyTransforms(
    ItemToModify=((Item=(Type=Bone,Name="base_shirt_front_l"),Transform=(Rotation=(X=0.0076,Y=-0.0868,Z=0.0868,W=0.9924),Translation=(X=-1,Y=0,Z=0),Scale3D=(X=1,Y=1,Z=1))))
    WeightMinimum=0
    Weight=&RigVMFunction_MathDoubleMul(
     B=&RigVMFunction_MathDoubleRemap(SourceMaximum=27.85, SourceMinimum=64.5, TargetMaximum=1, TargetMinimum=0, Value=@n4.Result).Result
     A=@n3.OutputParam
    ).Result
   )
   > n5 = &RigUnit_SphericalPoseReader(
    ActiveRegionScaleFactors=(PositiveWidth=1,NegativeWidth=1,PositiveHeight=1,NegativeHeight=1)
    ActiveRegionSize=0.35
    Debug=(bDrawDebug=true,bDraw2D=False,bDrawLocalAxes=False,DebugScale=25,DebugSegments=20,DebugThickness=0.25)
    DriverAxis=(X=1,Y=0,Z=0)
    DriverItem=(Type=Bone,Name="thigh_r")
    FalloffRegionScaleFactors=(PositiveWidth=1,NegativeWidth=1,PositiveHeight=1,NegativeHeight=1)
    FalloffSize=0.55
    FlipHeightScaling=false
    FlipWidthScaling=false
    OptionalParentItem=(Type=Bone,Name="None")
    RotationOffset=(X=-90,Y=0,Z=-20)
    Debug.bDrawDebug=@n2.Value
   )
   n6 = &RigVMFunction_MathVectorDistance(
    A=&RigUnit_GetTransform(Item=(Type=Bone,Name="base_shirt_front_r"), bInitial=False).Transform.Translation
    B=&RigUnit_GetTransform(Item=(Type=Bone,Name="calf_r"), bInitial=False).Transform.Translation
   )
   > &RigUnit_ModifyTransforms(
    ItemToModify=((Item=(Type=Bone,Name="base_shirt_front_r"),Transform=(Rotation=(X=-0.0076,Y=-0.0868,Z=-0.0868,W=0.9924),Translation=(X=-1,Y=0,Z=0),Scale3D=(X=1,Y=1,Z=1))))
    WeightMinimum=0
    Weight=&RigVMFunction_MathDoubleMul(
     B=&RigVMFunction_MathDoubleRemap(SourceMaximum=27.85, SourceMinimum=64.5, TargetMaximum=1, TargetMinimum=0, Value=@n6.Result).Result
     A=@n5.OutputParam
    ).Result
   )
   > &RigUnit_ModifyTransforms(
    ItemToModify="((Item=(Type=Bone,Name=\"base_short_back_l\"),Transform=(Rotation=(X=0,Y=0,Z=0,W=-1),Translation=(X=2.5,Y=-10,Z=0),Scale3D=(X=0.9,Y=1,Z=1.3))),(Item=(Type=Bone,Name=\"base_short_l\"),Transform=(Rotation=(X=0,Y=0,Z=0,W=1),Translation=(X=-6,Y=0,Z=0),Scale3D=(X=1 ...(truncated)"
    WeightMinimum=0
    Weight=RigVMDispatchNode(BoneName=root, Default=0, Name=deformRig_l_knee_alpha) {template="DISPATCH_RigDispatch_GetAnimAttribute(in Name,in BoneName,in Default,out Value,out Found)"}.Value
   )
   > &RigUnit_ModifyTransforms(
    ItemToModify="((Item=(Type=Bone,Name=\"base_short_back_r\"),Transform=(Rotation=(X=0,Y=0,Z=0,W=1),Translation=(X=2,Y=10,Z=-2),Scale3D=(X=0.9,Y=1,Z=1))),(Item=(Type=Bone,Name=\"base_short_r\"),Transform=(Rotation=(X=0,Y=0,Z=0,W=1),Translation=(X=-6,Y=0,Z=0.5),Scale3D=(X=1,Y= ...(truncated)"
    WeightMinimum=0
    Weight=RigVMDispatchNode(BoneName=root, Default=0, Name=deformRig_r_knee_alpha) {template="DISPATCH_RigDispatch_GetAnimAttribute(in Name,in BoneName,in Default,out Value,out Found)"}.Value
   )
   > &RigUnit_ModifyTransforms(
    ItemToModify=((Item=(Type=Bone,Name="dyn_short_l"),Transform=(Rotation=(X=0,Y=0,Z=0,W=1),Translation=(X=0,Y=0,Z=-2),Scale3D=(X=1.1,Y=1.1,Z=1.2))))
    WeightMinimum=0
    Weight=&RigVMFunction_MathDoubleMul(
     B=2
     A=RigVMDispatchNode(BoneName=root, Default=0, Name=deformRig_l_thigh_up_alpha) {template="DISPATCH_RigDispatch_GetAnimAttribute(in Name,in BoneName,in Default,out Value,out Found)"}.Value
    ).Result
   )
   > &RigUnit_ModifyTransforms(
    ItemToModify=((Item=(Type=Bone,Name="dyn_short_r"),Transform=(Rotation=(X=0,Y=0,Z=0,W=1),Translation=(X=0,Y=0,Z=-2),Scale3D=(X=1.1,Y=1.1,Z=1.2))))
    WeightMinimum=0
    Weight=&RigVMFunction_MathDoubleMul(
     B=2
     A=RigVMDispatchNode(BoneName=root, Default=0, Name=deformRig_r_thigh_up_alpha) {template="DISPATCH_RigDispatch_GetAnimAttribute(in Name,in BoneName,in Default,out Value,out Found)"}.Value
    ).Result
   )
   DeltaRotation = RigVMFunctionReferenceNode(
    Rotation Order=ZYX
    child=head
    debugRotation=false
    max=45
    min=30
    parent=head_null
    parentType=Null
    premult=1
    primAxis=2
    twistAxis=(X=1,Y=0,Z=0)
    ExecuteContext=@RigVMFunction_Sequence.F
   )
   > &RigUnit_ModifyTransforms(
    ItemToModify="((Item=(Type=Bone,Name=\"base_neck_frontmid_r\"),Transform=(Rotation=(X=0,Y=-0.0698,Z=0,W=0.9976),Translation=(X=0,Y=0,Z=0),Scale3D=(X=1,Y=1,Z=1))),(Item=(Type=Bone,Name=\"base_neck_mid_r\"),Transform=(Rotation=(X=0,Y=-0.0698,Z=0,W=0.9976),Translation=(X=0,Y=0 ...(truncated)"
    WeightMinimum=0
    Weight=@DeltaRotation.outputWeight
   )
   > DeltaRotation_1 = RigVMFunctionReferenceNode(
    Rotation Order=ZYX
    child=head
    debugRotation=false
    max=-18
    min=0
    parent=head_null
    parentType=Null
    premult=1
    primAxis=2
    twistAxis=(X=1,Y=0,Z=0)
   )
   > &RigUnit_ModifyTransforms(
    ItemToModify="((Item=(Type=Bone,Name=\"base_neck_frontmid_r\"),Transform=(Rotation=(X=0,Y=-0.0698,Z=0,W=0.9976),Translation=(X=0,Y=0,Z=2),Scale3D=(X=1,Y=1,Z=1))),(Item=(Type=Bone,Name=\"base_neck_frontmid_l\"),Transform=(Rotation=(X=0,Y=-0.0698,Z=0,W=0.9976),Translation=(X= ...(truncated)"
    WeightMinimum=0
    Weight=@DeltaRotation_1.outputWeight
   )
   > DeltaRotation_6 = RigVMFunctionReferenceNode(
    Rotation Order=ZYX
    child=head
    debugRotation=false
    max=-19
    min=-5
    parent=head_null
    parentType=Null
    premult=1
    primAxis=2
    twistAxis=(X=1,Y=0,Z=0)
   )
   > &RigUnit_ModifyTransforms(
    ItemToModify="((Item=(Type=Bone,Name=\"base_neck_frontmid_r\"),Transform=(Rotation=(X=0,Y=0.0698,Z=0,W=0.9976),Translation=(X=0,Y=0,Z=-2),Scale3D=(X=1,Y=1,Z=1))),(Item=(Type=Bone,Name=\"base_neck_frontmid_l\"),Transform=(Rotation=(X=0,Y=0.0698,Z=0,W=0.9976),Translation=(X=0 ...(truncated)"
    WeightMinimum=0
    Weight=@DeltaRotation_6.outputWeight
   )
   > DeltaRotation_3 = RigVMFunctionReferenceNode(
    Rotation Order=ZYX
    child=head
    debugRotation=false
    max=50
    min=25
    parent=head_null
    parentType=Null
    premult=1
    primAxis=1
    twistAxis=(X=1,Y=0,Z=0)
   )
   > &RigUnit_ModifyTransforms(
    ItemToModify="((Item=(Type=Bone,Name=\"base_neck_frontmid_l\"),Transform=(Rotation=(X=0,Y=-0.0436,Z=0,W=0.999),Translation=(X=0,Y=0,Z=0),Scale3D=(X=1,Y=1,Z=1))),(Item=(Type=Bone,Name=\"base_neck_mid_l\"),Transform=(Rotation=(X=0,Y=-0.0872,Z=0,W=0.9962),Translation=(X=0,Y=0, ...(truncated)"
    WeightMinimum=0
    Weight=@DeltaRotation_3.outputWeight
   )
   > DeltaRotation_2 = RigVMFunctionReferenceNode(
    Rotation Order=ZYX
    child=head
    debugRotation=false
    max=-50
    min=-25
    parent=head_null
    parentType=Null
    premult=1
    primAxis=1
    twistAxis=(X=1,Y=0,Z=0)
   )
   > &RigUnit_ModifyTransforms(
    ItemToModify="((Item=(Type=Bone,Name=\"base_neck_frontmid_r\"),Transform=(Rotation=(X=0,Y=-0.0436,Z=0,W=0.999),Translation=(X=0,Y=0,Z=0),Scale3D=(X=1,Y=1,Z=1))),(Item=(Type=Bone,Name=\"base_neck_mid_r\"),Transform=(Rotation=(X=0,Y=-0.0872,Z=0,W=0.9962),Translation=(X=0,Y=0, ...(truncated)"
    WeightMinimum=0
    Weight=@DeltaRotation_2.outputWeight
   )
   > DeltaRotation_4 = RigVMFunctionReferenceNode(
    Rotation Order=ZYX
    child=head
    debugRotation=false
    max=80
    min=25
    onlyTwist=true
    parent=head_null
    parentType=Null
    premult=1
    primAxis=0
    twistAxis=(X=1,Y=0,Z=0)
   )
   > &RigUnit_ModifyTransforms(
    ItemToModify="((Item=(Type=Bone,Name=\"base_neck_frontmid_r\"),Transform=(Rotation=(X=0,Y=-0.2588,Z=0,W=0.9659),Translation=(X=0,Y=0,Z=0),Scale3D=(X=1,Y=1,Z=1))),(Item=(Type=Bone,Name=\"base_neck_mid_r\"),Transform=(Rotation=(X=0,Y=-0.1736,Z=0,W=0.9848),Translation=(X=0,Y=0 ...(truncated)"
    WeightMinimum=0
    Weight=@DeltaRotation_4.outputWeight
   )
   > DeltaRotation_5 = RigVMFunctionReferenceNode(
    Rotation Order=ZYX
    child=head
    debugRotation=false
    max=-80
    min=-25
    onlyTwist=true
    parent=head_null
    parentType=Null
    premult=1
    primAxis=0
    twistAxis=(X=1,Y=0,Z=0)
   )
   > &RigUnit_ModifyTransforms(
    ItemToModify="((Item=(Type=Bone,Name=\"base_neck_frontmid_l\"),Transform=(Rotation=(X=0.017,Y=-0.1294,Z=0.1294,W=0.983),Translation=(X=0,Y=0,Z=0),Scale3D=(X=1,Y=1,Z=1))),(Item=(Type=Bone,Name=\"base_neck_mid_l\"),Transform=(Rotation=(X=0,Y=0,Z=0,W=1),Translation=(X=0,Y=0,Z= ...(truncated)"
    WeightMinimum=0
    Weight=@DeltaRotation_5.outputWeight
   )
   n7 = &RigVMFunction_MathBoolMake(Value=false)
   > n8 = &RigUnit_SphericalPoseReader(
    ActiveRegionScaleFactors=(PositiveWidth=1,NegativeWidth=0.5,PositiveHeight=0.75,NegativeHeight=0.75)
    ActiveRegionSize=0.1
    Debug=(bDrawDebug=false,bDraw2D=False,bDrawLocalAxes=False,DebugScale=25,DebugSegments=20,DebugThickness=0.25)
    DriverAxis=(X=0,Y=1,Z=0)
    DriverItem=(Type=Bone,Name="head")
    FalloffRegionScaleFactors=(PositiveWidth=1,NegativeWidth=0.5,PositiveHeight=0.5,NegativeHeight=0.5)
    FalloffSize=0.2
    FlipHeightScaling=false
    FlipWidthScaling=true
    OptionalParentItem=(Type=Bone,Name="None")
    RotationOffset=(X=100,Y=0,Z=0)
    Debug.bDrawDebug=@n7.Value
   )
   RigVMFunction_MathBoolNot = &RigVMFunction_MathBoolNot(Value=RigVMVariableNode() {variable=Crouching}.Value)
   Multiply_3 = RigVMAggregateNode(
    B=&RigVMFunction_MathBoolToFloat(Value=RigVMVariableNode() {variable=Targeting}.Value).Result
    C=&RigVMFunction_MathBoolToFloat(Value=@RigVMFunction_MathBoolNot.Result).Result
    A=@n8.OutputParam
   )
   AlphaInterp = &RigVMFunction_AlphaInterp(InterpSpeedDecreasing=12, InterpSpeedIncreasing=12, Scale=1, bClampResult=true, bMapRange=true, Value=@Multiply_3.Result)
   > &RigUnit_ModifyTransforms(
    ItemToModify="((Item=(Type=Bone,Name=\"base_neck_frontmid_l\"),Transform=(Rotation=(X=0,Y=-0.1736,Z=0,W=0.9848),Translation=(X=0,Y=3,Z=5),Scale3D=(X=1,Y=1,Z=1))),(Item=(Type=Bone,Name=\"base_neck_mid_l\"),Transform=(Rotation=(X=0,Y=-0.3007,Z=0,W=0.9537),Translation=(X=0,Y=0 ...(truncated)"
    WeightMinimum=0
    Weight=@AlphaInterp.Result
   )
   > n9 = &RigUnit_SphericalPoseReader(
    ActiveRegionScaleFactors=(PositiveWidth=1,NegativeWidth=1,PositiveHeight=1,NegativeHeight=1)
    ActiveRegionSize=0.05
    Debug=(bDrawDebug=false,bDraw2D=False,bDrawLocalAxes=False,DebugScale=25,DebugSegments=20,DebugThickness=0.25)
    DriverAxis=(X=0,Y=1,Z=0)
    DriverItem=(Type=Bone,Name="head")
    FalloffRegionScaleFactors=(PositiveWidth=1,NegativeWidth=1,PositiveHeight=1,NegativeHeight=1)
    FalloffSize=0.1
    FlipHeightScaling=False
    FlipWidthScaling=False
    OptionalParentItem=(Type=Bone,Name="None")
    RotationOffset=(X=95,Y=0,Z=0)
    Debug.bDrawDebug=@n7.Value
   )
   n10 = RigVMVariableNode() {variable=Crouching}
   n11 = &RigVMFunction_AlphaInterp(
    InterpSpeedDecreasing=12
    InterpSpeedIncreasing=12
    Scale=1
    bClampResult=true
    bMapRange=true
    Value=&RigVMFunction_MathDoubleMul(B=&RigVMFunction_MathBoolToFloat(Value=@n10.Value).Result, A=@n9.OutputParam).Result
   )
   > &RigUnit_ModifyTransforms(
    ItemToModify=((Item=(Type=Bone,Name="base_neck_frontmid_l"),Transform=(Rotation=(X=0,Y=-0.1736,Z=0,W=0.9848),Translation=(X=0,Y=0,Z=0),Scale3D=(X=1,Y=1,Z=1))))
    WeightMinimum=0
    Weight=@n11.Result
   )
   > n12 = &RigUnit_SphericalPoseReader(
    ActiveRegionScaleFactors=(PositiveWidth=1,NegativeWidth=0.5,PositiveHeight=0.75,NegativeHeight=0.75)
    ActiveRegionSize=0.1
    Debug=(bDrawDebug=true,bDraw2D=False,bDrawLocalAxes=False,DebugScale=25,DebugSegments=20,DebugThickness=0.25)
    DriverAxis=(X=0,Y=1,Z=0)
    DriverItem=(Type=Bone,Name="head")
    FalloffRegionScaleFactors=(PositiveWidth=1,NegativeWidth=0.5,PositiveHeight=1,NegativeHeight=1)
    FalloffSize=0.2
    FlipHeightScaling=False
    FlipWidthScaling=False
    OptionalParentItem=(Type=Bone,Name="None")
    RotationOffset=(X=100,Y=0,Z=-35)
    Debug.bDrawDebug=@n7.Value
   )
   Multiply_8 = RigVMAggregateNode(
    C=&RigVMFunction_MathBoolToFloat(Value=RigVMVariableNode() {variable=Crouching}.Value).Result
    B=RigVMVariableNode() {variable=CrouchingButUp}.Value
    A=@n12.OutputParam
   )
   n13 = &RigVMFunction_AlphaInterp(InterpSpeedDecreasing=12, InterpSpeedIncreasing=12, Scale=1, bClampResult=true, bMapRange=true, Value=@Multiply_8.Result)
   > &RigUnit_ModifyTransforms(
    ItemToModify="((Item=(Type=Bone,Name=\"base_neck_backmid_l\"),Transform=(Rotation=(X=0,Y=-0.6428,Z=0,W=0.766),Translation=(X=0,Y=0,Z=0),Scale3D=(X=1,Y=1,Z=1))),(Item=(Type=Bone,Name=\"base_neck_back\"),Transform=(Rotation=(X=0,Y=-0.7071,Z=0,W=0.7071),Translation=(X=-3,Y=0,Z ...(truncated)"
    WeightMinimum=0
    Weight=@n13.Result
   )
   > n14 = &RigUnit_SphericalPoseReader(
    ActiveRegionScaleFactors=(PositiveWidth=1,NegativeWidth=0.5,PositiveHeight=0.75,NegativeHeight=0.75)
    ActiveRegionSize=0.1
    Debug=(bDrawDebug=true,bDraw2D=False,bDrawLocalAxes=False,DebugScale=25,DebugSegments=20,DebugThickness=0.25)
    DriverAxis=(X=0,Y=1,Z=0)
    DriverItem=(Type=Bone,Name="head")
    FalloffRegionScaleFactors=(PositiveWidth=1,NegativeWidth=0.1,PositiveHeight=1,NegativeHeight=1)
    FalloffSize=0.2
    FlipHeightScaling=False
    FlipWidthScaling=False
    OptionalParentItem=(Type=Bone,Name="None")
    RotationOffset=(X=100,Y=0,Z=-30)
    Debug.bDrawDebug=@n7.Value
   )
   Multiply_7 = RigVMAggregateNode(
    C=&RigVMFunction_MathBoolToFloat(Value=RigVMVariableNode() {variable=Crouching}.Value).Result
    B=&RigVMFunction_MathBoolToFloat(Value=RigVMVariableNode() {variable=Targeting}.Value).Result
    D=RigVMVariableNode() {variable=CrouchingButUp}.Value
    A=@n14.OutputParam
   )
   > &RigUnit_ModifyTransforms(
    ItemToModify="((Item=(Type=Bone,Name=\"base_neck_backmid_l\"),Transform=(Rotation=(X=0,Y=0.5736,Z=0,W=0.8192),Translation=(X=0,Y=0,Z=0),Scale3D=(X=1,Y=1,Z=1))),(Item=(Type=Bone,Name=\"base_neck_mid_l\"),Transform=(Rotation=(X=0,Y=0.342,Z=0,W=0.9397),Translation=(X=0,Y=0,Z=0 ...(truncated)"
    WeightMinimum=0
    Weight=@Multiply_7.Result
   )
   RigVMFunction_MathBoolMake = &RigVMFunction_MathBoolMake(Value=false)
   n15 = &RigUnit_SphericalPoseReader(
    ActiveRegionScaleFactors=(PositiveWidth=1,NegativeWidth=1,PositiveHeight=1,NegativeHeight=1)
    ActiveRegionSize=0.05
    Debug=(bDrawDebug=true,bDraw2D=False,bDrawLocalAxes=False,DebugScale=25,DebugSegments=20,DebugThickness=0.25)
    DriverAxis=(X=1,Y=0,Z=0)
    DriverItem=(Type=Bone,Name="thigh_r")
    FalloffRegionScaleFactors=(PositiveWidth=1,NegativeWidth=1,PositiveHeight=1,NegativeHeight=1)
    FalloffSize=0.6
    FlipHeightScaling=true
    FlipWidthScaling=true
    OptionalParentItem=(Type=Bone,Name="None")
    RotationOffset=(X=-90,Y=0,Z=-20)
    Debug.bDrawDebug=@RigVMFunction_MathBoolMake.Value
    ExecutePin=@RigVMFunction_Sequence.D
   )
   > &RigUnit_ModifyTransforms(
    ItemToModify="((Item=(Type=Bone,Name=\"auto_skirt_front\"),Transform=(Rotation=(X=-0.4222,Y=-0.0184,Z=0.0395,W=0.9054),Translation=(X=-7,Y=9,Z=0),Scale3D=(X=1,Y=1,Z=1))),(Item=(Type=Bone,Name=\"auto_skirt_midfront\"),Transform=(Rotation=(X=-0.5,Y=0,Z=0,W=0.866),Translation= ...(truncated)"
    WeightMinimum=0
    Weight=@n15.OutputParam
   )
   > n16 = &RigUnit_SphericalPoseReader(
    ActiveRegionScaleFactors=(PositiveWidth=1,NegativeWidth=1,PositiveHeight=1,NegativeHeight=1)
    ActiveRegionSize=0.05
    Debug=(bDrawDebug=false,bDraw2D=False,bDrawLocalAxes=False,DebugScale=25,DebugSegments=20,DebugThickness=0.25)
    DriverAxis=(X=1,Y=0,Z=0)
    DriverItem=(Type=Bone,Name="thigh_r")
    FalloffRegionScaleFactors=(PositiveWidth=1,NegativeWidth=1,PositiveHeight=1,NegativeHeight=1)
    FalloffSize=0.55
    FlipHeightScaling=true
    FlipWidthScaling=true
    OptionalParentItem=(Type=Bone,Name="None")
    RotationOffset=(X=-40,Y=20,Z=0)
    Debug.bDrawDebug=@RigVMFunction_MathBoolMake.Value
   )
   > &RigUnit_ModifyTransforms(
    ItemToModify="((Item=(Type=Bone,Name=\"auto_skirt_front\"),Transform=(Rotation=(X=-0.2988,Y=-0.5792,Z=0.4056,W=0.6409),Translation=(X=-8.5,Y=0,Z=0),Scale3D=(X=1,Y=1,Z=1))),(Item=(Type=Bone,Name=\"auto_skirt_midfront\"),Transform=(Rotation=(X=0.3182,Y=0.4595,Z=-0.0918,W=-0.8 ...(truncated)"
    WeightMinimum=0
    Weight=@n16.OutputParam
   )
   > n17 = &RigUnit_SphericalPoseReader(
    ActiveRegionScaleFactors=(PositiveWidth=1,NegativeWidth=1,PositiveHeight=1,NegativeHeight=1)
    ActiveRegionSize=0.05
    Debug=(bDrawDebug=false,bDraw2D=False,bDrawLocalAxes=False,DebugScale=25,DebugSegments=20,DebugThickness=0.25)
    DriverAxis=(X=1,Y=0,Z=0)
    DriverItem=(Type=Bone,Name="thigh_r")
    FalloffRegionScaleFactors=(PositiveWidth=1,NegativeWidth=1,PositiveHeight=1,NegativeHeight=1)
    FalloffSize=0.6
    FlipHeightScaling=true
    FlipWidthScaling=true
    OptionalParentItem=(Type=Bone,Name="None")
    RotationOffset=(X=0,Y=20,Z=0)
    Debug.bDrawDebug=@RigVMFunction_MathBoolMake.Value
   )
   > &RigUnit_ModifyTransforms(
    ItemToModify="((Item=(Type=Bone,Name=\"auto_skirt_front\"),Transform=(Rotation=(X=0.0149,Y=-0.3417,Z=0.041,W=0.9388),Translation=(X=0,Y=0,Z=0),Scale3D=(X=1,Y=1,Z=1))),(Item=(Type=Bone,Name=\"auto_skirt_midfront\"),Transform=(Rotation=(X=0.035,Y=-0.5859,Z=-0.0733,W=0.8063),T ...(truncated)"
    WeightMinimum=0
    Weight=@n17.OutputParam
   )
   > n18 = &RigUnit_SphericalPoseReader(
    ActiveRegionScaleFactors=(PositiveWidth=1,NegativeWidth=1,PositiveHeight=1,NegativeHeight=1)
    ActiveRegionSize=0.05
    Debug=(bDrawDebug=false,bDraw2D=False,bDrawLocalAxes=False,DebugScale=25,DebugSegments=20,DebugThickness=0.25)
    DriverAxis=(X=1,Y=0,Z=0)
    DriverItem=(Type=Bone,Name="thigh_r")
    FalloffRegionScaleFactors=(PositiveWidth=1,NegativeWidth=1,PositiveHeight=1,NegativeHeight=1)
    FalloffSize=0.55
    FlipHeightScaling=true
    FlipWidthScaling=true
    OptionalParentItem=(Type=Bone,Name="None")
    RotationOffset=(X=40,Y=20,Z=0)
    Debug.bDrawDebug=@RigVMFunction_MathBoolMake.Value
   )
   > &RigUnit_ModifyTransforms(
    ItemToModify="((Item=(Type=Bone,Name=\"auto_skirt_front\"),Transform=(Rotation=(X=0,Y=0,Z=0,W=1),Translation=(X=0,Y=0,Z=0),Scale3D=(X=1,Y=1,Z=1))),(Item=(Type=Bone,Name=\"auto_skirt_midfront\"),Transform=(Rotation=(X=0,Y=0,Z=0,W=1),Translation=(X=0,Y=0,Z=0),Scale3D=(X=1,Y=1 ...(truncated)"
    WeightMinimum=0
    Weight=@n18.OutputParam
   )
   > n19 = &RigUnit_SphericalPoseReader(
    ActiveRegionScaleFactors=(PositiveWidth=1,NegativeWidth=1,PositiveHeight=1,NegativeHeight=1)
    ActiveRegionSize=0.05
    Debug=(bDrawDebug=false,bDraw2D=False,bDrawLocalAxes=False,DebugScale=25,DebugSegments=20,DebugThickness=0.25)
    DriverAxis=(X=1,Y=0,Z=0)
    DriverItem=(Type=Bone,Name="thigh_r")
    FalloffRegionScaleFactors=(PositiveWidth=1,NegativeWidth=1,PositiveHeight=1,NegativeHeight=1)
    FalloffSize=0.55
    FlipHeightScaling=true
    FlipWidthScaling=true
    OptionalParentItem=(Type=Bone,Name="None")
    RotationOffset=(X=90,Y=0,Z=20)
    Debug.bDrawDebug=@RigVMFunction_MathBoolMake.Value
   )
   > &RigUnit_ModifyTransforms(
    ItemToModify="((Item=(Type=Bone,Name=\"auto_skirt_front\"),Transform=(Rotation=(X=0,Y=0,Z=0,W=1),Translation=(X=0,Y=0,Z=0),Scale3D=(X=1,Y=1,Z=1))),(Item=(Type=Bone,Name=\"auto_skirt_midfront\"),Transform=(Rotation=(X=0,Y=0,Z=0,W=1),Translation=(X=0,Y=0,Z=0),Scale3D=(X=1,Y=1 ...(truncated)"
    WeightMinimum=0
    Weight=@n19.OutputParam
   )
   > n20 = &RigUnit_SphericalPoseReader(
    ActiveRegionScaleFactors=(PositiveWidth=1,NegativeWidth=1,PositiveHeight=1,NegativeHeight=1)
    ActiveRegionSize=0.05
    Debug=(bDrawDebug=true,bDraw2D=False,bDrawLocalAxes=False,DebugScale=25,DebugSegments=20,DebugThickness=0.25)
    DriverAxis=(X=-1,Y=0,Z=0)
    DriverItem=(Type=Bone,Name="thigh_l")
    FalloffRegionScaleFactors=(PositiveWidth=1,NegativeWidth=1,PositiveHeight=1,NegativeHeight=1)
    FalloffSize=0.55
    FlipHeightScaling=False
    FlipWidthScaling=False
    OptionalParentItem=(Type=Bone,Name="None")
    RotationOffset=(X=-90,Y=0,Z=20)
    Debug.bDrawDebug=@RigVMFunction_MathBoolMake.Value
   )
   > &RigUnit_ModifyTransforms(
    ItemToModify=((Item=(Type=Bone,Name="base_daggers_back"),Transform=(Rotation=(X=-0.0189,Y=0.0851,Z=0.2156,W=0.9726),Translation=(X=0,Y=0,Z=0),Scale3D=(X=1,Y=1,Z=1))))
    WeightMinimum=0
    Weight=@n20.OutputParam
   )
   RigUnit_SetTransform = &RigUnit_SetTransform(
    ExecutePin=()
    Item=(Type=Bone,Name="head")
    Weight=1
    bInitial=False
    Value=&RigUnit_GetTransform(Item=(Type=Control,Name="head_ctrl"), bInitial=False).Transform
   )
   &RigUnit_SetTransform(
    ExecutePin=()
    Item=(Type=Bone,Name="thigh_r")
    Weight=1
    bInitial=False
    Value=&RigUnit_GetTransform(Item=(Type=Control,Name="thigh_r_ctrl"), bInitial=False).Transform
   )
  ]}
]
hierarchy [
 root {loc=<0 0 0>, rot=<0 0 0 1>}
 pelvis {parent=root, loc=<0 -0.7071 93.1191>, rot=<0.0127 0.707 -0.0127 -0.707>}
 spine_01 {parent=pelvis, loc=<6.7494 -0.6388 0>, rot=<0 0 0.034 -0.9994>}
 spine_02 {parent=spine_01, loc=<7.2518 -0.035 0>, rot=<0 0 0.0238 -0.9997>}
 spine_03 {parent=spine_02, loc=<6.0104 0 0>, rot=<0 0 0.0149 -0.9999>}
 spine_04 {parent=spine_03, loc=<8.1024 0 0>, rot=<0 0 -0.0132 -0.9999>}
 spine_05 {parent=spine_04, loc=<10.0146 0 0>, rot=<0 0 -0.0015 -1>}
 neck_01 {parent=spine_05, loc=<11.3138 0.1529 0>, rot=<0 0 -0.2003 -0.9797>}
 neck_02 {parent=neck_01, loc=<4.5627 0 0>, rot=<0 0 0.0504 -0.9987>}
 head {parent=neck_02, loc=<5.3914 0 0>, rot=<0 0 0.115 -0.9934>}
 attach_cape {parent=spine_05, loc=<9.017 -11.714 0>, rot=<0.0616 -0.7044 -0.0616 -0.7044>}
 cape {parent=attach_cape, type=Null, loc=<0 0 0>, rot=<0 0 0 1>}
 attach_backpack {parent=spine_05, loc=<-0.123 -13.473 0>, rot=<-0.0093 -0.707 0.0093 -0.707>}
 Backpack_BR {parent=attach_backpack, type=Null, loc=<0 0 0>, rot=<0 0 0 1>}
 clavicle_l {parent=spine_05, loc=<4.3575 -0.1466 -2.3223>, rot=<0.0465 -0.7289 -0.009 -0.6829>}
 upperarm_l {parent=clavicle_l, loc=<15.5111 0 0>, rot=<0.0226 -0.3234 -0.0088 -0.9459>}
 lowerarm_l {parent=upperarm_l, loc=<26.6514 0 0>, rot=<0.0391 -0.0099 -0.2453 -0.9686>}
 hand_l {parent=lowerarm_l, loc=<22.3491 0 0>, rot=<-0.7081 -0.0507 0.0975 -0.6975>}
 thumb_01_l {parent=hand_l, loc=<1.3118 -0.3191 -2.459>, rot=<0.6546 -0.4042 0.0664 -0.6354>}
 thumb_02_l {parent=thumb_01_l, loc=<5.9336 -0.0448 0.0401>, rot=<-0.0037 0.0213 0.298 -0.9543>}
 thumb_03_l {parent=thumb_02_l, loc=<2.2223 0.3682 -0.0907>, rot=<0.0118 -0.013 0.0066 -0.9998>}
 index_metacarpal_l {parent=hand_l, loc=<1.6535 0.0183 -2.96>, rot=<0.0158 -0.0124 -0.0173 -0.9996>}
 index_01_l {parent=index_metacarpal_l, loc=<8.3008 0.0357 -0.0524>, rot=<-0.0044 0.0485 0.0924 -0.9945>}
 index_02_l {parent=index_01_l, loc=<3.3587 -0.3004 0.0452>, rot=<-0.002 0.0027 0.3225 -0.9466>}
 index_03_l {parent=index_02_l, loc=<2.5578 -0.123 -0.0252>, rot=<-0.002 0.0067 0.2558 -0.9667>}
 middle_metacarpal_l {parent=hand_l, loc=<2.1856 -0.1725 -0.5984>, rot=<0.006 0.0168 -0.0329 -0.9993>}
 middle_01_l {parent=middle_metacarpal_l, loc=<7.3275 -0.2447 0.0129>, rot=<-0.0208 0.0042 0.0858 -0.9961>}
 middle_02_l {parent=middle_01_l, loc=<4.0123 0.1401 0.1175>, rot=<-0.0218 0.0258 0.3512 -0.9357>}
 middle_03_l {parent=middle_02_l, loc=<2.6077 0.0534 0.1373>, rot=<0.0082 0.0188 0.1708 -0.9851>}
 ring_metacarpal_l {parent=hand_l, loc=<2.2074 -0.3625 1.3638>, rot=<-0.0284 0.0722 -0.0458 -0.9959>}
 ring_01_l {parent=ring_metacarpal_l, loc=<7.0295 -0.6197 -0.1446>, rot=<-0.0489 -0.0252 0.1892 -0.9804>}
 ring_02_l {parent=ring_01_l, loc=<3.6174 0.4647 -0.1027>, rot=<0 0 0.2806 -0.9598>}
 ring_03_l {parent=ring_02_l, loc=<2.4152 0.114 0.0199>, rot=<0 0 0.1189 -0.9929>}
 pinky_metacarpal_l {parent=hand_l, loc=<1.7616 -0.9753 3.3814>, rot=<-0.0397 0.0814 -0.069 -0.9935>}
 pinky_01_l {parent=pinky_metacarpal_l, loc=<7.1709 -0.4831 -0.1094>, rot=<-0.0654 -0.0035 0.2242 -0.9723>}
 pinky_02_l {parent=pinky_01_l, loc=<2.8274 0.0893 -0.1043>, rot=<0.0037 0.0011 0.2854 -0.9584>}
 pinky_03_l {parent=pinky_02_l, loc=<2.2046 0.0307 -0.0078>, rot=<-0.0002 0.0098 0.1445 -0.9895>}
 deform_wrist_dn_base_l {parent=hand_l, loc=<0 -2.2342 0>, rot=<0 0 0 -1>}
 deform_wrist_dn_l {parent=deform_wrist_dn_base_l, loc=<0 0 0>, rot=<0 0 0 1>}
 deform_wrist_up_base_l {parent=hand_l, loc=<0 2.2682 0>, rot=<0 0 0 -1>}
 deform_wrist_up_l {parent=deform_wrist_up_base_l, loc=<0 0 0>, rot=<0 0 0 1>}
 lowerarm_twist_02_l {parent=lowerarm_l, loc=<9.6199 0 0>, rot=<0 0 0 -1>}
 lowerarm_twist_01_l {parent=lowerarm_l, loc=<17.8549 0 0>, rot=<0 0 0 -1>}
 deform_elbow_in_base_l {parent=lowerarm_l, loc=<0.2112 5.259 0.8274>, rot=<0 0 0 1>}
 deform_elbow_in_l {parent=deform_elbow_in_base_l, loc=<0 0 0>, rot=<0 0 0 1>}
 deform_elbow_out_base_l {parent=lowerarm_l, loc=<0 -4 0>, rot=<0 0 0 1>}
 deform_elbow_out_l {parent=deform_elbow_out_base_l, loc=<0 0 0>, rot=<0 0 0 1>}
 upperarm_twist_01_l {parent=upperarm_l, loc=<0.0041 0 0>, rot=<0 0 0 -1>}
 deform_upperarm_bk_base_l {parent=upperarm_twist_01_l, loc=<1.0813 -7.2193 -1.9453>, rot=<0 0 0 1>}
 deform_upperarm_bk_l {parent=deform_upperarm_bk_base_l, loc=<0 0 0>, rot=<0 0 0 1>}
 deform_upperarm_fr_base_l {parent=upperarm_twist_01_l, loc=<-0.3691 7.0506 -2.765>, rot=<0 0 0 1>}
 deform_upperarm_fr_l {parent=deform_upperarm_fr_base_l, loc=<0 0 0>, rot=<0 0 0 1>}
 deform_upperarm_up_base_l {parent=upperarm_twist_01_l, loc=<-3.2216 -1.1391 3.3311>, rot=<0 0 0 1>}
 deform_upperarm_up_l {parent=deform_upperarm_up_base_l, loc=<0 0 0>, rot=<0 0 0 1>}
 upperarm_twist_02_l {parent=upperarm_l, loc=<13.7941 0 0>, rot=<0 0 0 -1>}
 base_sleeve_l {parent=upperarm_twist_02_l, loc=<-1.0432 -0.1082 -1.0402>, rot=<0.0028 0.0524 -0.0314 -0.9981>}
 dyn_sleeve_l {parent=base_sleeve_l, loc=<0 0 0>, rot=<0 0 0 1>}
 clavicle_r {parent=spine_05, loc=<4.3575 -0.1466 2.3223>, rot=<0.7289 0.0465 -0.6829 0.009>}
 upperarm_r {parent=clavicle_r, loc=<-15.5111 0 0>, rot=<0.0226 -0.3234 -0.0088 -0.9459>}
 lowerarm_r {parent=upperarm_r, loc=<-26.6514 0 0>, rot=<0.0391 -0.0099 -0.2453 -0.9686>}
 hand_r {parent=lowerarm_r, loc=<-22.3491 0 0>, rot=<-0.7081 -0.0507 0.0975 -0.6975>}
 thumb_01_r {parent=hand_r, loc=<-1.3118 0.3191 2.459>, rot=<0.6546 -0.4042 0.0664 -0.6354>}
 thumb_02_r {parent=thumb_01_r, loc=<-5.9336 0.0448 -0.0401>, rot=<-0.0037 0.0213 0.298 -0.9543>}
 thumb_03_r {parent=thumb_02_r, loc=<-2.2223 -0.3682 0.0907>, rot=<0.0118 -0.013 0.0066 -0.9998>}
 index_metacarpal_r {parent=hand_r, loc=<-1.6535 -0.0183 2.96>, rot=<0.0158 -0.0124 -0.0173 -0.9996>}
 index_01_r {parent=index_metacarpal_r, loc=<-8.3008 -0.0357 0.0524>, rot=<-0.0044 0.0485 0.0924 -0.9945>}
 index_02_r {parent=index_01_r, loc=<-3.3587 0.3004 -0.0452>, rot=<-0.002 0.0027 0.3225 -0.9466>}
 index_03_r {parent=index_02_r, loc=<-2.5578 0.123 0.0252>, rot=<-0.002 0.0067 0.2558 -0.9667>}
 middle_metacarpal_r {parent=hand_r, loc=<-2.1856 0.1725 0.5984>, rot=<0.006 0.0168 -0.0329 -0.9993>}
 middle_01_r {parent=middle_metacarpal_r, loc=<-7.3275 0.2447 -0.0129>, rot=<-0.0208 0.0042 0.0858 -0.9961>}
 middle_02_r {parent=middle_01_r, loc=<-4.0123 -0.1401 -0.1175>, rot=<-0.0218 0.0258 0.3512 -0.9357>}
 middle_03_r {parent=middle_02_r, loc=<-2.6077 -0.0534 -0.1373>, rot=<0.0082 0.0188 0.1708 -0.9851>}
 ring_metacarpal_r {parent=hand_r, loc=<-2.2074 0.3625 -1.3638>, rot=<-0.0284 0.0722 -0.0458 -0.9959>}
 ring_01_r {parent=ring_metacarpal_r, loc=<-7.0295 0.6197 0.1446>, rot=<-0.0489 -0.0252 0.1892 -0.9804>}
 ring_02_r {parent=ring_01_r, loc=<-3.6174 -0.4647 0.1027>, rot=<0 0 0.2806 -0.9598>}
 ring_03_r {parent=ring_02_r, loc=<-2.4152 -0.114 -0.0199>, rot=<0 0 0.1189 -0.9929>}
 pinky_metacarpal_r {parent=hand_r, loc=<-1.7616 0.9753 -3.3814>, rot=<-0.0397 0.0814 -0.069 -0.9935>}
 pinky_01_r {parent=pinky_metacarpal_r, loc=<-7.1709 0.4831 0.1094>, rot=<-0.0654 -0.0035 0.2242 -0.9723>}
 pinky_02_r {parent=pinky_01_r, loc=<-2.8274 -0.0893 0.1043>, rot=<0.0037 0.0011 0.2854 -0.9584>}
 pinky_03_r {parent=pinky_02_r, loc=<-2.2046 -0.0307 0.0078>, rot=<-0.0002 0.0098 0.1445 -0.9895>}
 deform_wrist_dn_base_r {parent=hand_r, loc=<-0.0004 2.2345 -0.0001>, rot=<0 0 0 -1>}
 deform_wrist_dn_r {parent=deform_wrist_dn_base_r, loc=<0 0 0>, rot=<0 0 0 1>}
 deform_wrist_up_base_r {parent=hand_r, loc=<0.0003 -2.2684 0.0001>, rot=<0 0 0 -1>}
 deform_wrist_up_r {parent=deform_wrist_up_base_r, loc=<0 0 0>, rot=<0 0 0 1>}
 lowerarm_twist_02_r {parent=lowerarm_r, loc=<-9.6199 0 0>, rot=<0 0 0 -1>}
 lowerarm_twist_01_r {parent=lowerarm_r, loc=<-17.8549 0 0>, rot=<0 0 0 -1>}
 deform_elbow_in_base_r {parent=lowerarm_r, loc=<-0.2111 -5.2591 -0.8275>, rot=<0 0 0 1>}
 deform_elbow_in_r {parent=deform_elbow_in_base_r, loc=<-0.0001 0 0.0002>, rot=<0 0 0 1>}
 deform_elbow_out_base_r {parent=lowerarm_r, loc=<0 4 -0.0001>, rot=<0 0 0 1>}
 deform_elbow_out_r {parent=deform_elbow_out_base_r, loc=<0 0 0>, rot=<0 0 0 1>}
 upperarm_twist_01_r {parent=upperarm_r, loc=<-0.0041 0 0>, rot=<0 0 0 -1>}
 deform_upperarm_bk_base_r {parent=upperarm_twist_01_r, loc=<-1.0814 7.2192 1.9455>, rot=<0 0 0 1>}
 deform_upperarm_bk_r {parent=deform_upperarm_bk_base_r, loc=<0 0 0>, rot=<0 0 0 1>}
 deform_upperarm_fr_base_r {parent=upperarm_twist_01_r, loc=<0.3689 -7.0506 2.7652>, rot=<0 0 0 1>}
 deform_upperarm_fr_r {parent=deform_upperarm_fr_base_r, loc=<0 0 0>, rot=<0 0 0 1>}
 deform_upperarm_up_base_r {parent=upperarm_twist_01_r, loc=<3.2214 1.1391 -3.3308>, rot=<0 0 0 1>}
 deform_upperarm_up_r {parent=deform_upperarm_up_base_r, loc=<0 0 0>, rot=<0 0 0 -1>}
 upperarm_twist_02_r {parent=upperarm_r, loc=<-13.7941 0 0>, rot=<0 0 0 -1>}
 base_sleeve_r {parent=upperarm_twist_02_r, loc=<1.0437 0.1205 1.033>, rot=<0.0324 -0.9981 0.0036 -0.0518>}
 dyn_sleeve_r {parent=base_sleeve_r, loc=<0 0 0>, rot=<0 0 0 1>}
 base_strap_back_r {parent=spine_05, loc=<-2.8976 -13.51 5.0062>, rot=<-0.0059 -0.6674 0.7447 -0.0047>}
 dyn_strap_back_r_01 {parent=base_strap_back_r, loc=<0 0 0>, rot=<0 0 0 1>}
 dyn_strap_back_r_02 {parent=dyn_strap_back_r_01, loc=<3.3841 -0.0031 0.0001>, rot=<0.0102 0.022 0.0023 -0.9997>}
 base_strap_back_l {parent=spine_05, loc=<-2.8978 -13.5298 -5.0121>, rot=<0.0047 -0.7428 0.6695 0.0059>}
 dyn_strap_back_l_01 {parent=base_strap_back_l, loc=<0 0 0>, rot=<0 0 0 1>}
 dyn_strap_back_l_02 {parent=dyn_strap_back_l_01, loc=<3.3832 0.0027 0.0002>, rot=<-0.0102 0.0219 -0.0024 -0.9997>}
 base_neck_mid_r {parent=spine_05, loc=<11.2826 2.6708 7.2666>, rot=<-0.0106 -0.0209 -0.1668 -0.9857>}
 dyn_neck_mid_r {parent=base_neck_mid_r, loc=<0 0 0>, rot=<0 0 0 1>}
 base_neck_frontmid_r {parent=spine_05, loc=<6.8949 8.048 5.2329>, rot=<0.3344 0.0544 -0.1715 -0.9251>}
 dyn_neck_frontmid_r {parent=base_neck_frontmid_r, loc=<0 0 0>, rot=<0 0 0 1>}
 base_neck_mid_l {parent=spine_05, loc=<10.7204 4.3026 -6.9241>, rot=<0.9882 0.1529 -0.0027 0.0053>}
 dyn_neck_mid_l {parent=base_neck_mid_l, loc=<0 0 0>, rot=<0 0 0 1>}
 base_neck_frontmid_l {parent=spine_05, loc=<6.9816 9.0109 -4.0786>, rot=<0.8777 0.1741 -0.1013 -0.4348>}
 dyn_neck_frontmid_l {parent=base_neck_frontmid_l, loc=<0 0 0>, rot=<0 0 0 1>}
 base_neck_backmid_r {parent=spine_05, loc=<12.5873 -1.2954 5.3568>, rot=<-0.4518 0.011 -0.0975 -0.8867>}
 dyn_neck_backmid_r {parent=base_neck_backmid_r, loc=<0 0 0>, rot=<0 0 0 1>}
 base_neck_backmid_l {parent=spine_05, loc=<12.5638 -1.3598 -5.2295>, rot=<0.9039 0.1162 -0.0155 0.4114>}
 dyn_neck_backmid_l {parent=base_neck_backmid_l, loc=<0 0 0>, rot=<0 0 0 1>}
 head_null {parent=spine_05, type=Null, loc=<20.6573 3.5489 0>, rot=<0 0 -0.036 -0.9994>}
 base_neck_back {parent=spine_05, loc=<13.5253 -3.16 0>, rot=<-0.7065 -0.0284 -0.0284 -0.7065>}
 dyn_neck_back {parent=base_neck_back, loc=<0 0 0>, rot=<0 0 0 1>}
 deform_pec_aim_base_l {parent=spine_04, loc=<7.8732 12.6388 -2.4123>, rot=<0.0102 -0.4602 0.3979 -0.7936>}
 deform_pec_aim_l {parent=deform_pec_aim_base_l, loc=<0 0 0>, rot=<0 0 0 1>}
 deform_pec_base_l {parent=deform_pec_aim_l, loc=<10.766 0 0>, rot=<0 0 0 1>}
 deform_pec_l {parent=deform_pec_base_l, loc=<0.055 0 0>, rot=<0 0 0 1>}
 deform_pec_aim_base_r {parent=spine_04, loc=<7.8727 12.6388 2.4123>, rot=<-0.4602 -0.0102 0.7936 0.3979>}
 deform_pec_aim_r {parent=deform_pec_aim_base_r, loc=<0 0 0>, rot=<0 0 0 1>}
 deform_pec_base_r {parent=deform_pec_aim_r, loc=<-10.7662 -0.0006 -0.0007>, rot=<0 0 0 1>}
 deform_pec_r {parent=deform_pec_base_r, loc=<0 0 0>, rot=<0 0 0 1>}
 base_strap_front_l {parent=spine_03, loc=<3.0922 14.5571 -9.669>, rot=<0.0412 -0.8921 -0.4484 0.0373>}
 dyn_strap_front_l_01 {parent=base_strap_front_l, loc=<0 0 0>, rot=<0 0 0 1>}
 dyn_strap_front_l_02 {parent=dyn_strap_front_l_01, loc=<5.2759 0.0022 0>, rot=<-0.0095 -0.0028 0.0057 -0.9999>}
 dyn_strap_front_l_03 {parent=dyn_strap_front_l_02, loc=<6.1429 0.0002 0.0374>, rot=<-0.0014 -0.0095 0.0031 -0.9999>}
 base_shirt_front_l {parent=spine_03, loc=<-1.8101 10.9073 -8.2004>, rot=<-0.1863 -0.8523 -0.4828 0.0762>}
 dyn_shirt_front_l {parent=base_shirt_front_l, loc=<0 0 0>, rot=<0 0 0 1>}
 base_shirt_front_r {parent=spine_03, loc=<-1.7799 11.1599 8.2547>, rot=<-0.0669 0.5413 0.8203 0.1725>}
 dyn_shirt_front_r {parent=base_shirt_front_r, loc=<0 0 0>, rot=<0 0 0 1>}
 base_shirt_back {parent=spine_03, loc=<5.9391 -12.0491 0>, rot=<0.0387 -0.706 0.706 -0.0387>}
 dyn_shirt_back {parent=base_shirt_back, loc=<0 0 0>, rot=<0 0 0 1>}
 base_shirt_mid_l {parent=spine_03, loc=<1.3574 0.1356 -16.6789>, rot=<0.0547 -0.9985 0 0>}
 dyn_shirt_mid_l {parent=base_shirt_mid_l, loc=<0 0 0>, rot=<0 0 0 1>}
 base_shirt_mid_r {parent=spine_03, loc=<1.3574 0.1356 16.6789>, rot=<0 0 0.9985 -0.0547>}
 dyn_shirt_mid_r {parent=base_shirt_mid_r, loc=<0 0 0>, rot=<0 0 0 1>}
 base_daggers_back {parent=spine_01, loc=<4.2881 -12.8405 -4.2631>, rot=<0.1155 -0.6914 0.7011 -0.1309>}
 dyn_daggers_back {parent=base_daggers_back, loc=<0 0 0>, rot=<0 0 0 1>}
 thigh_l {parent=pelvis, loc=<-2.5693 -0.0109 -7.7501>, rot=<0.0012 0.0572 0.0155 -0.9982>}
 calf_l {parent=thigh_l, loc=<-39.0484 0 0>, rot=<0.0006 -0.0161 -0.0391 -0.9991>}
 foot_l {parent=calf_l, loc=<-42.561 0.1077 0.4292>, rot=<0.0418 -0.0455 0.0343 -0.9975>}
 ball_l {parent=foot_l, loc=<-6.1599 13.8264 0.6806>, rot=<0.0001 -0.0001 -0.7063 -0.7079>}
 calf_twist_02_l {parent=calf_l, loc=<-18.8064 -0.0967 0.0397>, rot=<0.0237 0.0095 -0.0035 -0.9997>}
 calf_twist_01_l {parent=calf_l, loc=<-35.813 -0.0967 0.0397>, rot=<0.0237 0.0095 -0.0035 -0.9997>}
 deform_knee_in_base_l {parent=calf_l, loc=<1.298 -5.2957 -0.0496>, rot=<0 0 0 -1>}
 deform_knee_in_l {parent=deform_knee_in_base_l, loc=<0 0 0>, rot=<0 0 0 1>}
 base_short_back_l {parent=deform_knee_in_l, loc=<8.5866 -6.827 -1.3412>, rot=<-0.0302 -0.9991 -0.0042 0.03>}
 dyn_short_back_l {parent=base_short_back_l, loc=<0 0 0>, rot=<0 0 0 1>}
 deform_knee_out_base_l {parent=calf_l, loc=<0 4 0>, rot=<0 0 0 -1>}
 deform_knee_out_l {parent=deform_knee_out_base_l, loc=<0 0 0>, rot=<0 0 0 1>}
 thigh_twist_01_l {parent=thigh_l, loc=<-21.6303 0 0>, rot=<0.0013 0 0 -1>}
 base_short_l {parent=thigh_twist_01_l, loc=<-10.1949 -0.175 -1.6255>, rot=<0.0089 -0.9989 -0.0032 0.0462>}
 dyn_short_l {parent=base_short_l, loc=<0 0 0>, rot=<0 0 0 1>}
 deform_groin_base_l {parent=thigh_l, loc=<-2.1908 8.1142 -1.7442>, rot=<0 0 0 -1>}
 deform_groin_l {parent=deform_groin_base_l, loc=<0 0 0>, rot=<0 0 0 1>}
 thigh_r {parent=pelvis, loc=<-2.5693 -0.0109 7.7501>, rot=<0.0572 -0.0012 0.9982 0.0155>}
 calf_r {parent=thigh_r, loc=<39.0484 0 0>, rot=<0.0006 -0.0161 -0.0391 -0.9991>}
 foot_r {parent=calf_r, loc=<42.561 -0.1077 -0.4292>, rot=<0.0418 -0.0455 0.0343 -0.9975>}
 ball_r {parent=foot_r, loc=<6.1599 -13.8264 -0.6806>, rot=<0.0001 -0.0001 -0.7063 -0.7079>}
 calf_twist_02_r {parent=calf_r, loc=<18.8064 0.0967 -0.0397>, rot=<0.0237 0.0095 -0.0035 -0.9997>}
 calf_twist_01_r {parent=calf_r, loc=<35.813 0.0967 -0.0397>, rot=<0.0237 0.0095 -0.0035 -0.9997>}
 deform_knee_in_base_r {parent=calf_r, loc=<-1.298 5.2957 0.0496>, rot=<0 0 0 1>}
 deform_knee_in_r {parent=deform_knee_in_base_r, loc=<0 0 0>, rot=<0 0 0 1>}
 base_short_back_r {parent=deform_knee_in_r, loc=<-8.8348 6.8571 -0.2453>, rot=<-0.002 0.0168 0.0353 -0.9992>}
 dyn_short_back_r {parent=base_short_back_r, loc=<0 0 0>, rot=<0 0 0 1>}
 deform_knee_out_base_r {parent=calf_r, loc=<0 -4 0>, rot=<0 0 0 -1>}
 deform_knee_out_r {parent=deform_knee_out_base_r, loc=<0 0 0>, rot=<0 0 0 1>}
 thigh_twist_01_r {parent=thigh_r, loc=<21.6303 0 0>, rot=<0.0013 0 0 -1>}
 base_short_r {parent=thigh_twist_01_r, loc=<10.4853 0.2588 0.9568>, rot=<-0.0027 0.0007 -0.0037 -1>}
 dyn_short_r {parent=base_short_r, loc=<0 0 0>, rot=<0 0 0 1>}
 deform_groin_base_r {parent=thigh_r, loc=<2.1908 -8.1142 1.7442>, rot=<0 0 0 -1>}
 deform_groin_r {parent=deform_groin_base_r, loc=<0 0 0>, rot=<0 0 0 1>}
 deform_glute_base_l {parent=pelvis, loc=<-8.8064 -8.8222 -8.4287>, rot=<0.0275 0.0501 -0.4472 -0.8926>}
 deform_glute_l {parent=deform_glute_base_l, loc=<0 0 0>, rot=<0 0 0 1>}
 deform_glute_base_r {parent=pelvis, loc=<-8.8062 -8.8222 8.4287>, rot=<0.0501 -0.0275 0.8926 -0.4472>}
 deform_glute_r {parent=deform_glute_base_r, loc=<0 0 0>, rot=<0 0 0 1>}
 base_skirt_mid {parent=pelvis, loc=<3.3215 -2.5567 15.8463>, rot=<0.2188 -0.0408 0.9733 -0.0566>}
 auto_skirt_mid {parent=base_skirt_mid, loc=<0 0 0>, rot=<0 0 0 1>}
 offset_skirt_mid {parent=auto_skirt_mid, loc=<0 0 0>, rot=<0 0 0 1>}
 dyn_skirt_mid_01 {parent=offset_skirt_mid, loc=<0 0 0>, rot=<0 0 0 1>}
 dyn_skirt_mid_02 {parent=dyn_skirt_mid_01, loc=<10.9748 -0.0003 -0.0246>, rot=<0.0028 -0.149 0.0578 -0.9871>}
 dyn_skirt_mid_03 {parent=dyn_skirt_mid_02, loc=<12.1018 0.1276 -0.0699>, rot=<0 0.051 0 -0.9987>}
 dyn_skirt_mid_04 {parent=dyn_skirt_mid_03, loc=<12.4924 0 0>, rot=<0 -0.0177 0 -0.9998>}
 base_strap_skirtbag {parent=dyn_skirt_mid_02, loc=<0.6304 0.4555 1.3722>, rot=<0.0921 0.064 -0.0081 -0.9937>}
 dyn_strap_skirtbag_01 {parent=base_strap_skirtbag, loc=<0 0 0>, rot=<0 0 0 1>}
 dyn_strap_skirtbag_02 {parent=dyn_strap_skirtbag_01, loc=<5.6379 -0.3297 -0.048>, rot=<0 0 0 1>}
 base_skirt_back {parent=pelvis, loc=<2.8819 -12.5827 6.6854>, rot=<0.1436 -0.6033 0.7823 0.0581>}
 auto_skirt_back {parent=base_skirt_back, loc=<0 0 0>, rot=<0 0 0 1>}
 offset_skirt_back {parent=auto_skirt_back, loc=<0 0 0>, rot=<0 0 0 1>}
 dyn_skirt_back_01 {parent=offset_skirt_back, loc=<0 0 0>, rot=<0 0 0 1>}
 dyn_skirt_back_02 {parent=dyn_skirt_back_01, loc=<11.14 0 0>, rot=<-0.0009 -0.0248 -0.0372 -0.999>}
 dyn_skirt_back_03 {parent=dyn_skirt_back_02, loc=<11.3729 -0.0686 0.0678>, rot=<0.0042 -0.0104 -0.04 -0.9991>}
 dyn_skirt_back_04 {parent=dyn_skirt_back_03, loc=<12.332 0.007 -0.0308>, rot=<-0.0048 0.004 -0.0165 -0.9998>}
 base_skirt_front {parent=pelvis, loc=<2.023 7.9635 13.3336>, rot=<0.1256 0.5697 0.8118 0.0253>}
 auto_skirt_front {parent=base_skirt_front, loc=<0 0 0>, rot=<0 0 0 1>}
 offset_skirt_front {parent=auto_skirt_front, loc=<0 0 0>, rot=<0 0 0 1>}
 dyn_skirt_front_01 {parent=offset_skirt_front, loc=<0 0 0>, rot=<0 0 0 1>}
 dyn_skirt_front_02 {parent=dyn_skirt_front_01, loc=<9.291 0.0412 -0.0065>, rot=<0.0021 -0.0558 0.0212 -0.9982>}
 dyn_skirt_front_03 {parent=dyn_skirt_front_02, loc=<10.9556 0.038 -0.0464>, rot=<-0.0001 -0.0187 -0.0029 -0.9998>}
 dyn_skirt_front_04 {parent=dyn_skirt_front_03, loc=<12.2813 0.1319 0.1023>, rot=<0 -0.002 0.0167 -0.9999>}
 base_skirt_midback {parent=pelvis, loc=<3.0315 -8.8679 12.6673>, rot=<0.1924 -0.3903 0.8995 0.0376>}
 auto_skirt_midback {parent=base_skirt_midback, loc=<0 0 0>, rot=<0 0 0 1>}
 offset_skirt_midback {parent=auto_skirt_midback, loc=<0 0 0>, rot=<0 0 0 1>}
 dyn_skirt_midback_01 {parent=offset_skirt_midback, loc=<0 0 0>, rot=<0 0 0 1>}
 dyn_skirt_midback_02 {parent=dyn_skirt_midback_01, loc=<9.8779 0.0003 0.1105>, rot=<-0.0042 -0.0766 -0.0545 -0.9956>}
 dyn_skirt_midback_03 {parent=dyn_skirt_midback_02, loc=<12.9553 0 0>, rot=<0 0 0 1>}
 dyn_skirt_midback_04 {parent=dyn_skirt_midback_03, loc=<12.4218 -0.0029 -0.0006>, rot=<0.0001 0.0043 -0.0181 -0.9998>}
 base_skirt_midfront {parent=pelvis, loc=<3.0136 2.7448 15.5736>, rot=<0.2131 0.2402 0.9464 -0.0353>}
 auto_skirt_midfront {parent=base_skirt_midfront, loc=<0 0 0>, rot=<0 0 0 1>}
 offset_skirt_midfront {parent=auto_skirt_midfront, loc=<0 0 0>, rot=<0 0 0 1>}
 dyn_skirt_midfront_01 {parent=offset_skirt_midfront, loc=<0 0 0>, rot=<0 0 0 1>}
 dyn_skirt_midfront_02 {parent=dyn_skirt_midfront_01, loc=<11.6893 0.0353 0.0165>, rot=<-0.0109 -0.1455 0.0742 -0.9865>}
 dyn_skirt_midfront_03 {parent=dyn_skirt_midfront_02, loc=<10.7388 -0.024 -0.0651>, rot=<0 0.0397 0 -0.9992>}
 dyn_skirt_midfront_04 {parent=dyn_skirt_midfront_03, loc=<12.3296 0.0622 0.0034>, rot=<0 -0.0038 0 -1>}
 head_ctrl {controlType=EULER_TRANSFORM, type=Control, loc=<0 0 0>, rot=<0 0 0 1>}
 thigh_r_ctrl {controlType=EULER_TRANSFORM, type=Control, loc=<0 0 0>, rot=<0 0 0 1>}
]
variables [
 Crouching {type=bool, default=False}
 CrouchingButUp {type=double, default=0}
 Targeting {type=bool, default=False}
]
