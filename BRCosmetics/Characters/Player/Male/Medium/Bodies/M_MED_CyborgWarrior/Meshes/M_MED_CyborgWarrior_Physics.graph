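# Physics asset for the male medium CyborgWarrior cosmetic body, covering its dangling skirt, straps, skirt bag, back daggers, shirt and shorts pieces.
# Five four-segment skirt chains hang from pelvis offset bodies, with swing limits that generally widen toward the hem.
# Straps and daggers hang limited from base bodies, mostly free-rotating on spine or pelvis.
# Capsules also cover torso, thighs, right calf and foot; free-angular constraints cross-link some skirt panels and the bag.
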PhysicsAsset /BRCosmetics/Characters/Player/Male/Medium/Bodies/M_MED_CyborgWarrior/Meshes/M_MED_CyborgWarrior_Physics
bodies {
 base_daggers_back [{shapeType=Capsule, length=0.505, radius=0.505, shapeName=base_daggers_back_capsule, rotation=<rot 0 -90 0>}]
 base_strap_back_l [{shapeType=Capsule, length=0.505, radius=0.505, shapeName=base_strap_back_l_capsule, rotation=<rot 0 -90 0>}]
 base_strap_back_r [{shapeType=Capsule, length=0.505, radius=0.505, shapeName=base_strap_back_r_capsule, rotation=<rot 0 -90 0>}]
 base_strap_front_l [{shapeType=Capsule, length=0.505, radius=0.505, shapeName=base_strap_front_l_capsule, rotation=<rot 0 -90 0>}]
 base_strap_skirtbag [{shapeType=Capsule, length=0.505, radius=0.505, shapeName=base_strap_skirtbag_capsule, rotation=<rot 0 -90 0>}]
 calf_r [{shapeType=Capsule, length=40, radius=10, shapeName=calf_r_capsule, center=<25 0 0>, rotation=<rot 90 0 0>}]
 dyn_daggers_back [{shapeType=Capsule, length=6, radius=6, shapeName=dyn_daggers_back_capsule, center=<0 0 3.6>, rotation=<rot 85 180 180>}]
 dyn_shirt_front_l [{shapeType=Capsule, length=11, radius=3.5, shapeName=dyn_shirt_front_l_capsule, center=<4.5868 0 -2.4>, rotation=<rot 90 -90 0>}]
 dyn_short_r [{shapeType=Capsule, length=16, radius=13.5, shapeName=dyn_short_r_capsule, center=<5.9948 0.25 -1.2>, rotation=<rot 90 0 0>}]
 dyn_skirt_back_01 [{shapeType=Capsule, length=9, radius=4.5, shapeName=dyn_skirt_back_01_capsule, center=<4.0414 0 2.7947>, rotation=<rot 90 0 0>}]
 dyn_skirt_back_02 [{shapeType=Capsule, length=16, radius=7.3, shapeName=dyn_skirt_back_02_capsule, center=<9.7972 1.4386 3.4775>, rotation=<rot 90 -0.0003 0>}]
 dyn_skirt_back_03 [{shapeType=Capsule, length=19, radius=10, shapeName=dyn_skirt_back_03_capsule, center=<11.1078 0 6.821>, rotation=<rot 0 0 90>}]
 dyn_skirt_back_04 [{
   shapeType=Capsule
   length=11
   radius=10
   shapeName=dyn_skirt_back_04_capsule
   center=<6.1555 0.2232 6.2581>
   rotation=<rot -0.0003 -0.0001 90>
  }]
 dyn_skirt_front_01 [{shapeType=Capsule, length=9, radius=4.5, shapeName=dyn_skirt_front_01_capsule, center=<4.0414 -2.8199 2.7947>, rotation=<rot 90 0 0>}]
 dyn_skirt_front_02 [{shapeType=Capsule, length=11, radius=5.5, shapeName=dyn_skirt_front_02_capsule, center=<7.5592 -1.9571 3.4775>, rotation=<rot 90 -0.0006 0>}]
 dyn_skirt_front_03 [{shapeType=Capsule, length=29, radius=9.8, shapeName=dyn_skirt_front_03_capsule, center=<11.1078 -1.6216 6.821>, rotation=<rot 0 3 90>}]
 dyn_skirt_front_04 [{shapeType=Capsule, length=29, radius=9.8, shapeName=dyn_skirt_front_04_capsule, center=<12.834 0.2119 7.2664>, rotation=<rot 0 0 90>}]
 dyn_skirt_mid_01 [{shapeType=Capsule, length=9, radius=4.5, shapeName=dyn_skirt_mid_01_capsule, center=<4.0414 -1.2741 2.7947>, rotation=<rot 90 0 0>}]
 dyn_skirt_mid_02 [{shapeType=Capsule, length=11, radius=5.5, shapeName=dyn_skirt_mid_02_capsule, center=<7.5592 -1.2742 4>, rotation=<rot 88 -0.0006 0>}]
 dyn_skirt_mid_03 [{shapeType=Capsule, length=22, radius=9.5, shapeName=dyn_skirt_mid_03_capsule, center=<11.1078 -1.068 7.2>, rotation=<rot 0 0 90>}]
 dyn_skirt_mid_04 [{shapeType=Capsule, length=22, radius=9.5, shapeName=dyn_skirt_mid_04_capsule, center=<7.976 -0.0653 7>, rotation=<rot 0 0 90>}]
 dyn_skirt_midback_01 [{shapeType=Capsule, length=9, radius=4.5, shapeName=dyn_skirt_midback_01_capsule, center=<4.0414 -0.8324 2.7947>, rotation=<rot 90 0 0>}]
 dyn_skirt_midback_02 [{
   shapeType=Capsule
   length=11
   radius=5.5
   shapeName=dyn_skirt_midback_02_capsule
   center=<7.5592 -0.5697 3.4775>
   rotation=<rot 90 -0.0005 0>
  }]
 dyn_skirt_midback_03 [{shapeType=Capsule, length=19, radius=9, shapeName=dyn_skirt_midback_03_capsule, center=<11.1078 0 6.821>, rotation=<rot 0 0 90>}]
 dyn_skirt_midback_04 [{shapeType=Capsule, length=11, radius=9, shapeName=dyn_skirt_midback_04_capsule, center=<6.8316 -0.0328 6.5098>, rotation=<rot 0 0 90>}]
 dyn_skirt_midfront_01 [{shapeType=Capsule, length=9, radius=4.5, shapeName=dyn_skirt_midfront_01_capsule, center=<4.0414 -2.1505 2.7947>, rotation=<rot 90 0 0>}]
 dyn_skirt_midfront_02 [{
   shapeType=Capsule
   length=11
   radius=5.5
   shapeName=dyn_skirt_midfront_02_capsule
   center=<7.5591 -2.1505 3.4775>
   rotation=<rot 90 -0.0005 0>
  }]
 dyn_skirt_midfront_03 [{shapeType=Capsule, length=29, radius=9, shapeName=dyn_skirt_midfront_03_capsule, center=<11.1078 -2.1035 6.821>, rotation=<rot 0 0 90>}]
 dyn_skirt_midfront_04 [{shapeType=Capsule, length=29, radius=9, shapeName=dyn_skirt_midfront_04_capsule, center=<10.1256 -0.0653 6.7235>, rotation=<rot 0 0 90>}]
 dyn_strap_back_l_01 [{shapeType=Capsule, length=4, radius=2, shapeName=dyn_strap_back_l_01_capsule, center=<0 0 1>, rotation=<rot 90 0 0>}]
 dyn_strap_back_l_02 [{shapeType=Capsule, length=4, radius=2, shapeName=dyn_strap_back_l_02_capsule, center=<0.75 0 1>, rotation=<rot 90 0 0>}]
 dyn_strap_back_r_01 [{shapeType=Capsule, length=4, radius=2, shapeName=dyn_strap_back_r_01_capsule, center=<0 0 1>, rotation=<rot 90 0 0>}]
 dyn_strap_back_r_02 [{shapeType=Capsule, length=4, radius=2, shapeName=dyn_strap_back_r_02_capsule, center=<0.75 0 1>, rotation=<rot 90 0 0>}]
 dyn_strap_front_l_01 [{shapeType=Capsule, length=6.5, radius=0.8, shapeName=dyn_strap_front_l_01_capsule, center=<2.3961 0 0>, rotation=<rot 90 0 0>}]
 dyn_strap_front_l_02 [{shapeType=Capsule, length=6.5, radius=0.8, shapeName=dyn_strap_front_l_02_capsule, center=<2.6979 0 0>, rotation=<rot 90 0 0>}]
 dyn_strap_front_l_03 [{shapeType=Capsule, length=3, radius=0.8, shapeName=dyn_strap_front_l_03_capsule, center=<1.2971 0 0>, rotation=<rot 90 0 0>}]
 dyn_strap_skirtbag_01 [{
   shapeType=Capsule
   length=6
   radius=2
   shapeName=dyn_strap_skirtbag_01_capsule
   center=<3.3565 -1.1386 9.5663>
   rotation=<rot -0.0001 -0.0004 90.0001>
  }]
 dyn_strap_skirtbag_02 [{shapeType=Capsule, length=3, radius=2.1, shapeName=dyn_strap_skirtbag_02_capsule, center=<2.4722 -1.1386 8.824>, rotation=<rot 0 0 90>}]
 foot_r [{shapeType=Capsule, length=9, radius=8.5, shapeName=foot_l_capsule, center=<4.6505 2.3232 -2.3367>}]
 offset_skirt_back [{shapeType=Capsule, length=0.505, radius=0.505, shapeName=offset_skirt_back_capsule, rotation=<rot 0 -90 0>}]
 offset_skirt_front [{shapeType=Capsule, length=0.505, radius=0.505, shapeName=offset_skirt_front_capsule, rotation=<rot 0 -90 0>}]
 offset_skirt_mid [{shapeType=Capsule, length=0.505, radius=0.505, shapeName=offset_skirt_mid_capsule, rotation=<rot 0 -90 0>}]
 offset_skirt_midback [{shapeType=Capsule, length=0.505, radius=0.505, shapeName=offset_skirt_midback_capsule, rotation=<rot 0 -90 0>}]
 offset_skirt_midfront [{shapeType=Capsule, length=0.505, radius=0.505, shapeName=offset_skirt_midfront_capsule, rotation=<rot 0 -90 0>}]
 pelvis [{shapeType=Capsule, length=6, radius=11, shapeName=pelvis_capsule}]
 spine_02 [{shapeType=Capsule, length=11, radius=9.5, shapeName=spine_02_capsule, center=<3.3813 0.5921 0.0156>, rotation=<rot 1.3859 -90 -0.1455>}]
 spine_04 [{shapeType=Capsule, length=22, radius=12.8, shapeName=spine_04_capsule, center=<-1.5 0 0>}]
 thigh_l [{shapeType=Capsule, length=33, radius=10.5, shapeName=thigh_l_capsule, center=<-10 0 0>, rotation=<rot 90 -5 0>}]
 thigh_r [{shapeType=Capsule, length=44, radius=11, shapeName=thigh_r_capsule, center=<22 0 0>, rotation=<rot 90 -5 0>}]
}
constraints [
 {bone1Name=dyn_skirt_back_01, bone2Name=offset_skirt_back, swing1LimitDegrees=3, swing2LimitDegrees=15, twistLimitDegrees=1}
 {bone1Name=dyn_skirt_back_02, bone2Name=dyn_skirt_back_01, swing1LimitDegrees=5, swing2LimitDegrees=20, twistLimitDegrees=1}
 {bone1Name=dyn_skirt_back_03, bone2Name=dyn_skirt_back_02, swing1LimitDegrees=8, swing2LimitDegrees=25, twistLimitDegrees=1}
 {bone1Name=dyn_skirt_back_04, bone2Name=dyn_skirt_back_03, swing1LimitDegrees=10, swing2LimitDegrees=30, twistLimitDegrees=1}
 {bone1Name=dyn_skirt_front_01, bone2Name=offset_skirt_front, swing1LimitDegrees=3, swing2LimitDegrees=15, twistLimitDegrees=1}
 {bone1Name=dyn_skirt_front_02, bone2Name=dyn_skirt_front_01, swing1LimitDegrees=10, swing2LimitDegrees=15, twistLimitDegrees=1}
 {bone1Name=dyn_skirt_front_03, bone2Name=dyn_skirt_front_02, swing1LimitDegrees=8, swing2LimitDegrees=20, twistLimitDegrees=10}
 {bone1Name=dyn_skirt_front_04, bone2Name=dyn_skirt_front_03, swing1LimitDegrees=10, swing2LimitDegrees=20, twistLimitDegrees=1}
 {bone1Name=dyn_skirt_mid_01, bone2Name=offset_skirt_mid, swing1LimitDegrees=3, swing2LimitDegrees=15, twistLimitDegrees=1}
 {bone1Name=dyn_skirt_mid_02, bone2Name=dyn_skirt_mid_01, swing1LimitDegrees=5, swing2LimitDegrees=20, twistLimitDegrees=1}
 {bone1Name=dyn_skirt_mid_03, bone2Name=dyn_skirt_mid_02, swing1LimitDegrees=8, swing2LimitDegrees=25, twistLimitDegrees=1}
 {bone1Name=dyn_skirt_mid_04, bone2Name=dyn_skirt_mid_03, swing1LimitDegrees=10, swing2LimitDegrees=30, twistLimitDegrees=1}
 {bone1Name=dyn_strap_skirtbag_01, bone2Name=base_strap_skirtbag, swing1LimitDegrees=5, swing2LimitDegrees=15, twistLimitDegrees=1}
 {bone1Name=dyn_strap_skirtbag_02, bone2Name=dyn_strap_skirtbag_01, swing1LimitDegrees=5, swing2LimitDegrees=30, twistLimitDegrees=1}
 {bone1Name=dyn_skirt_midback_01, bone2Name=offset_skirt_midback, swing1LimitDegrees=3, swing2LimitDegrees=15, twistLimitDegrees=1}
 {bone1Name=dyn_skirt_midback_02, bone2Name=dyn_skirt_midback_01, swing1LimitDegrees=5, swing2LimitDegrees=20, twistLimitDegrees=1}
 {bone1Name=dyn_skirt_midback_03, bone2Name=dyn_skirt_midback_02, swing1LimitDegrees=8, swing2LimitDegrees=25, twistLimitDegrees=1}
 {bone1Name=dyn_skirt_midback_04, bone2Name=dyn_skirt_midback_03, swing1LimitDegrees=10, swing2LimitDegrees=30, twistLimitDegrees=1}
 {bone1Name=dyn_skirt_midfront_01, bone2Name=offset_skirt_midfront, swing1LimitDegrees=3, swing2LimitDegrees=15, twistLimitDegrees=1}
 {bone1Name=dyn_skirt_midfront_02, bone2Name=dyn_skirt_midfront_01, swing1LimitDegrees=5, swing2LimitDegrees=20, twistLimitDegrees=1}
 {bone1Name=dyn_skirt_midfront_03, bone2Name=dyn_skirt_midfront_02, swing1LimitDegrees=8, swing2LimitDegrees=25, twistLimitDegrees=1}
 {bone1Name=dyn_skirt_midfront_04, bone2Name=dyn_skirt_midfront_03, swing1LimitDegrees=10, swing2LimitDegrees=30, twistLimitDegrees=1}
 {
  bone1Name=dyn_short_r
  bone2Name=thigh_r
  swing1LimitDegrees=45
  swing1Motion=Free
  swing2LimitDegrees=45
  swing2Motion=Free
  twistLimitDegrees=45
  twistMotion=Free
 }
 {bone1Name=dyn_daggers_back, bone2Name=base_daggers_back, swing1LimitDegrees=10, swing2LimitDegrees=20, twistLimitDegrees=1}
 {bone1Name=dyn_strap_front_l_03, bone2Name=dyn_strap_front_l_02, swing1LimitDegrees=20, swing2LimitDegrees=20, twistLimitDegrees=1}
 {bone1Name=dyn_strap_front_l_01, bone2Name=base_strap_front_l, swing1LimitDegrees=8, swing2LimitDegrees=8, twistLimitDegrees=1}
 {bone1Name=dyn_strap_front_l_02, bone2Name=dyn_strap_front_l_01, swing1LimitDegrees=15, swing2LimitDegrees=15, twistLimitDegrees=1}
 {bone1Name=dyn_strap_back_l_01, bone2Name=base_strap_back_l, swing1LimitDegrees=10, swing2LimitDegrees=15, twistLimitDegrees=1}
 {bone1Name=dyn_strap_back_l_02, bone2Name=dyn_strap_back_l_01, swing1LimitDegrees=15, swing2LimitDegrees=25, twistLimitDegrees=1}
 {bone1Name=dyn_strap_back_r_01, bone2Name=base_strap_back_r, swing1LimitDegrees=10, swing2LimitDegrees=15, twistLimitDegrees=1}
 {bone1Name=dyn_strap_back_r_02, bone2Name=dyn_strap_back_r_01, swing1LimitDegrees=15, swing2LimitDegrees=25, twistLimitDegrees=1}
 {
  bone1Name=base_strap_back_r
  bone2Name=spine_04
  swing1LimitDegrees=45
  swing1Motion=Free
  swing2LimitDegrees=45
  swing2Motion=Free
  twistLimitDegrees=45
  twistMotion=Free
 }
 {
  bone1Name=base_strap_back_l
  bone2Name=spine_04
  swing1LimitDegrees=45
  swing1Motion=Free
  swing2LimitDegrees=45
  swing2Motion=Free
  twistLimitDegrees=45
  twistMotion=Free
 }
 {
  bone1Name=base_strap_front_l
  bone2Name=spine_02
  swing1LimitDegrees=45
  swing1Motion=Free
  swing2LimitDegrees=45
  swing2Motion=Free
  twistLimitDegrees=45
  twistMotion=Free
 }
 {
  bone1Name=base_daggers_back
  bone2Name=pelvis
  swing1LimitDegrees=45
  swing1Motion=Free
  swing2LimitDegrees=45
  swing2Motion=Free
  twistLimitDegrees=45
  twistMotion=Free
 }
 {
  bone1Name=thigh_l
  bone2Name=pelvis
  swing1LimitDegrees=45
  swing1Motion=Free
  swing2LimitDegrees=45
  swing2Motion=Free
  twistLimitDegrees=45
  twistMotion=Free
 }
 {
  bone1Name=thigh_r
  bone2Name=pelvis
  swing1LimitDegrees=45
  swing1Motion=Free
  swing2LimitDegrees=45
  swing2Motion=Free
  twistLimitDegrees=45
  twistMotion=Free
 }
 {
  bone1Name=spine_02
  bone2Name=pelvis
  swing1LimitDegrees=45
  swing1Motion=Free
  swing2LimitDegrees=45
  swing2Motion=Free
  twistLimitDegrees=45
  twistMotion=Free
 }
 {
  bone1Name=spine_04
  bone2Name=spine_02
  swing1LimitDegrees=45
  swing1Motion=Free
  swing2LimitDegrees=45
  swing2Motion=Free
  twistLimitDegrees=45
  twistMotion=Free
 }
 {
  bone1Name=calf_r
  bone2Name=thigh_r
  swing1LimitDegrees=45
  swing1Motion=Free
  swing2LimitDegrees=45
  swing2Motion=Free
  twistLimitDegrees=45
  twistMotion=Free
 }
 {
  bone1Name=offset_skirt_front
  bone2Name=pelvis
  swing1LimitDegrees=45
  swing1Motion=Free
  swing2LimitDegrees=45
  swing2Motion=Free
  twistLimitDegrees=45
  twistMotion=Free
 }
 {
  bone1Name=offset_skirt_mid
  bone2Name=pelvis
  swing1LimitDegrees=45
  swing1Motion=Free
  swing2LimitDegrees=45
  swing2Motion=Free
  twistLimitDegrees=45
  twistMotion=Free
 }
 {
  bone1Name=offset_skirt_midback
  bone2Name=pelvis
  swing1LimitDegrees=45
  swing1Motion=Free
  swing2LimitDegrees=45
  swing2Motion=Free
  twistLimitDegrees=45
  twistMotion=Free
 }
 {
  bone1Name=offset_skirt_midfront
  bone2Name=pelvis
  swing1LimitDegrees=45
  swing1Motion=Free
  swing2LimitDegrees=45
  swing2Motion=Free
  twistLimitDegrees=45
  twistMotion=Free
 }
 {
  bone1Name=offset_skirt_back
  bone2Name=pelvis
  swing1LimitDegrees=45
  swing1Motion=Free
  swing2LimitDegrees=45
  swing2Motion=Free
  twistLimitDegrees=45
  twistMotion=Free
 }
 {
  bone1Name=dyn_shirt_front_l
  bone2Name=spine_02
  swing1LimitDegrees=45
  swing1Motion=Free
  swing2LimitDegrees=45
  swing2Motion=Free
  twistLimitDegrees=45
  twistMotion=Free
 }
 {
  bone1Name=foot_r
  bone2Name=calf_r
  swing1LimitDegrees=45
  swing1Motion=Free
  swing2LimitDegrees=45
  swing2Motion=Free
  twistLimitDegrees=45
  twistMotion=Free
 }
 {
  bone1Name=base_strap_skirtbag
  bone2Name=dyn_skirt_mid_02
  swing1LimitDegrees=45
  swing1Motion=Free
  swing2LimitDegrees=45
  swing2Motion=Free
  twistLimitDegrees=45
  twistMotion=Free
 }
 {
  bone1Name=dyn_skirt_midfront_02
  bone2Name=base_strap_skirtbag
  swing1LimitDegrees=45
  swing1Motion=Free
  swing2LimitDegrees=45
  swing2Motion=Free
  twistLimitDegrees=45
  twistMotion=Free
 }
 {
  bone1Name=dyn_skirt_midback_02
  bone2Name=base_strap_skirtbag
  swing1LimitDegrees=45
  swing1Motion=Free
  swing2LimitDegrees=45
  swing2Motion=Free
  twistLimitDegrees=45
  twistMotion=Free
 }
 {
  bone1Name=dyn_skirt_front_04
  bone2Name=dyn_skirt_midfront_04
  swing1LimitDegrees=45
  swing1Motion=Free
  swing2LimitDegrees=45
  swing2Motion=Free
  twistLimitDegrees=45
  twistMotion=Free
 }
]
