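# Animation Blueprint for the CyborgWarrior player body cosmetic, copying the parent mesh pose and adding secondary motion to clothing and gear.
# Constraints, AnimDynamics, a Control Rig, additive bone offsets and RigidBody physics drive straps, daggers, shirt and skirt.
# Backpacks or capes disable back-piece dynamics, while front-end without emotes switches RigidBody base from pelvis to right thigh.
# EventGraph updates gravity, simulation settings and slide, crouch, movement and LOD states.
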AnimBlueprint /BRCosmetics/Characters/Player/Male/Medium/Bodies/M_MED_CyborgWarrior/Meshes/M_MED_CyborgWarrior_AnimBP
target_skeleton=&/M_MED_CyborgWarrior_Skeleton
graphs [
 AnimGraph {nodes [
   n1 = NOT Boolean(A=Get IsWearingBackpackOrCape() -> (IsWearingBackpackOrCape:bool))
   n2 = "Constraint - base_strap_back_l (PPPPP)"!(
    ComponentPose=Local To Component(LocalPose=Copy Pose From Mesh() {bUseAttachedParent=true})
    Alpha=1
    ConstraintWeights_0=0.4611
    ConstraintWeights_1:float=0.2759
    ConstraintWeights_2:float=0.1984
    ConstraintWeights_3:float=0.0546
    ConstraintWeights_4:float=0.0099
   ) -> (Pose) {
    alphaBoolBlend {blendOption=Linear}
    boneToModify {boneName=base_strap_back_l}
    constraintSetup [
     {targetBone {boneName=spine_03}}
     {targetBone {boneName=spine_04}}
     {targetBone {boneName=spine_05}}
     {targetBone {boneName=spine_02}}
     {targetBone {boneName=spine_01}}
    ]
    constraintWeights [1,1,1,1,1]
   }
   n3 = "Constraint - base_strap_back_r (PPPP)"!(
    ComponentPose=@n2
    Alpha=1
    ConstraintWeights_0=0.4844
    ConstraintWeights_1:float=0.3246
    ConstraintWeights_2:float=0.1365
    ConstraintWeights_3:float=0.0545
   ) -> (Pose) {
    alphaBoolBlend {blendOption=Linear}
    boneToModify {boneName=base_strap_back_r}
    constraintSetup [{targetBone {boneName=spine_03}}, {targetBone {boneName=spine_04}}, {targetBone {boneName=spine_05}}, {targetBone {boneName=spine_02}}]
    constraintWeights [1,1,1,1]
   }
   n4 = "Constraint - base_daggers_back (PPPPP)"!(
    ComponentPose=@n3
    Alpha=1
    ConstraintWeights_0=0.7641
    ConstraintWeights_1:float=0.0877
    ConstraintWeights_2:float=0.0866
    ConstraintWeights_3:float=0.0597
    ConstraintWeights_4:float=0.0019
   ) -> (Pose) {
    alphaBoolBlend {blendOption=Linear}
    boneToModify {boneName=base_daggers_back}
    constraintSetup [
     {targetBone {boneName=pelvis}}
     {targetBone {boneName=spine_02}}
     {targetBone {boneName=spine_01}}
     {targetBone {boneName=spine_03}}
     {targetBone {boneName=spine_04}}
    ]
    constraintWeights [1,1,1,1,1]
   }
   n5 = Anim Dynamics - Bone: dyn_sleeve_r!(ComponentPose=@n4, Alpha=1, GravityOverride:Vector="0,0,-980", ExternalForce="0,0,0") -> (Pose) {
    angularDampingOverride=1
    angularSpringConstant=11000
    bAngularSpring=true
    bOverrideAngularDamping=true
    bUseGravityOverride=true
    gravityScale=0.35
    simulationSpace=RootRelative
    alphaBoolBlend {blendOption=Linear}
    boundBone {boneName=dyn_sleeve_r}
    componentAppliedLinearAccClamp=<9500 9500 9500>
    componentLinearAccScale=<0.5 0.5 0.5>
    componentLinearVelScale=<0.5 0.5 0.5>
    gravityOverride {z=-980}
    physicsBodyDefinitions [{
      boundBone {boneName=dyn_sleeve_r}
      boxExtents=<20 30 30>
      constraintSetup {angularLimitsMax=<10 10 10>, angularLimitsMin=<-10 -10 -10>, angularTarget=<1 0 0>}
      localJointOffset=<4 0 0>
     }]
   }
   n6 = Anim Dynamics - Bone: dyn_sleeve_l!(ComponentPose=@n5, Alpha=1, GravityOverride:Vector="0,0,-980", ExternalForce="0,0,0") -> (Pose) {
    angularDampingOverride=1
    angularSpringConstant=11000
    bAngularSpring=true
    bOverrideAngularDamping=true
    bUseGravityOverride=true
    gravityScale=0.35
    simulationSpace=RootRelative
    alphaBoolBlend {blendOption=Linear}
    boundBone {boneName=dyn_sleeve_l}
    componentAppliedLinearAccClamp=<9500 9500 9500>
    componentLinearAccScale=<0.5 0.5 0.5>
    componentLinearVelScale=<0.5 0.5 0.5>
    gravityOverride {z=-980}
    physicsBodyDefinitions [{
      boundBone {boneName=dyn_sleeve_l}
      boxExtents=<20 30 30>
      constraintSetup {angularLimitsMax=<10 10 10>, angularLimitsMin=<-10 -10 -10>, angularTarget=<1 0 0>}
      localJointOffset=<4 0 0>
     }]
   }
   n7 = Anim Dynamics - Bone: dyn_short_r!(ComponentPose=@n6, Alpha=1, GravityOverride:Vector="0,0,-980", ExternalForce="0,0,0") -> (Pose) {
    angularDampingOverride=0.999
    angularSpringConstant=5000
    bAngularSpring=true
    bOverrideAngularDamping=true
    bUseGravityOverride=true
    gravityScale=0.35
    simulationSpace=RootRelative
    alphaBoolBlend {blendOption=Linear}
    boundBone {boneName=dyn_short_r}
    componentAppliedLinearAccClamp=<5000 5000 5000>
    componentLinearAccScale=<0.6 0.6 1>
    componentLinearVelScale=<1 1 1>
    gravityOverride {z=-980}
    physicsBodyDefinitions [{
      sphereCollisionRadius=12
      boundBone {boneName=dyn_short_r}
      boxExtents=<20 30 30>
      constraintSetup {angularLimitsMax=<10 10 10>, angularLimitsMin=<-10 -10 -10>, angularTarget=<1 0 0>, linearAxesMax=<0 2 0>, linearAxesMin=<-2 -1 0>}
      localJointOffset=<4 0 0>
     }]
   }
   n8 = Anim Dynamics - Bone: dyn_short_l!(ComponentPose=@n7, Alpha=1, GravityOverride:Vector="0,0,-980", ExternalForce="0,0,0") -> (Pose) {
    angularDampingOverride=0.999
    angularSpringConstant=5000
    bAngularSpring=true
    bOverrideAngularDamping=true
    bUseGravityOverride=true
    gravityScale=0.35
    simulationSpace=RootRelative
    alphaBoolBlend {blendOption=Linear}
    boundBone {boneName=dyn_short_l}
    componentAppliedLinearAccClamp=<5000 5000 5000>
    componentLinearAccScale=<0.6 0.6 1>
    componentLinearVelScale=<1 1 1>
    gravityOverride {z=-980}
    physicsBodyDefinitions [{
      sphereCollisionRadius=12
      boundBone {boneName=dyn_short_l}
      boxExtents=<20 30 30>
      constraintSetup {angularLimitsMax=<10 10 10>, angularLimitsMin=<-10 -10 -10>, angularTarget=<1 0 0>, linearAxesMax=<0 1 1>, linearAxesMin=<-1 -1 0>}
      localJointOffset=<4 0 0>
     }]
   }
   n9 = Anim Dynamics - Bone: dyn_shirt_front_r!(ComponentPose=@n8, Alpha=1, GravityOverride:Vector="0,0,-980", ExternalForce="0,0,0") -> (Pose) {
    angularDampingOverride=1
    angularSpringConstant=11000
    bAngularSpring=true
    bOverrideAngularDamping=true
    bUseGravityOverride=true
    gravityScale=0.35
    simulationSpace=RootRelative
    alphaBoolBlend {blendOption=Linear}
    boundBone {boneName=dyn_shirt_front_r}
    componentAppliedLinearAccClamp=<9500 9500 9500>
    componentLinearAccScale=<1 1 1>
    componentLinearVelScale=<1 1 1>
    gravityOverride {z=-980}
    physicsBodyDefinitions [{
      boundBone {boneName=dyn_shirt_front_r}
      boxExtents=<20 30 30>
      constraintSetup {angularLimitsMax=<3 20 15>, angularLimitsMin=<-3 -20 -15>, angularTarget=<1 0 0>}
      localJointOffset=<4 0 0>
     }]
   }
   n10 = Anim Dynamics - Bone: dyn_shirt_mid_r!(ComponentPose=@n9, Alpha=1, GravityOverride:Vector="0,0,-980", ExternalForce="0,0,0") -> (Pose) {
    angularDampingOverride=1
    angularSpringConstant=11000
    bAngularSpring=true
    bOverrideAngularDamping=true
    bUseGravityOverride=true
    gravityScale=0.35
    simulationSpace=RootRelative
    alphaBoolBlend {blendOption=Linear}
    boundBone {boneName=dyn_shirt_mid_r}
    componentAppliedLinearAccClamp=<9500 9500 9500>
    componentLinearAccScale=<1 1 1>
    componentLinearVelScale=<1 1 1>
    gravityOverride {z=-980}
    physicsBodyDefinitions [{
      boundBone {boneName=dyn_shirt_mid_r}
      boxExtents=<20 30 30>
      constraintSetup {angularLimitsMax=<2 20 10>, angularLimitsMin=<-2 -20 -10>, angularTarget=<1 0 0>}
      localJointOffset=<4 0 0>
     }]
   }
   n11 = Anim Dynamics - Bone: dyn_shirt_back!(ComponentPose=@n10, bAlphaBoolEnabled:bool=@n1, GravityOverride:Vector="0,0,-980", ExternalForce="0,0,0") -> (Pose) {
    alphaInputType=Bool
    angularDampingOverride=1
    angularSpringConstant=11000
    bAngularSpring=true
    bOverrideAngularDamping=true
    bUseGravityOverride=true
    gravityScale=0.35
    simulationSpace=RootRelative
    alphaBoolBlend {blendInTime=0.5, blendOption=QuadraticInOut, blendOutTime=1}
    boundBone {boneName=dyn_shirt_back}
    componentAppliedLinearAccClamp=<9500 9500 9500>
    componentLinearAccScale=<1 1 1>
    componentLinearVelScale=<1 1 1>
    gravityOverride {z=-980}
    physicsBodyDefinitions [{
      boundBone {boneName=dyn_shirt_back}
      boxExtents=<20 30 30>
      constraintSetup {angularLimitsMax=<1 20 5>, angularLimitsMin=<-1 -20 -5>, angularTarget=<1 0 0>}
      localJointOffset=<4 0 0>
     }]
   }
   n12 = Anim Dynamics - Bone: dyn_shirt_mid_l!(ComponentPose=@n11, Alpha=1, GravityOverride:Vector="0,0,-980", ExternalForce="0,0,0") -> (Pose) {
    angularDampingOverride=1
    angularSpringConstant=11000
    bAngularSpring=true
    bOverrideAngularDamping=true
    bUseGravityOverride=true
    gravityScale=0.35
    simulationSpace=RootRelative
    alphaBoolBlend {blendOption=Linear}
    boundBone {boneName=dyn_shirt_mid_l}
    componentAppliedLinearAccClamp=<9500 9500 9500>
    componentLinearAccScale=<1 1 1>
    componentLinearVelScale=<1 1 1>
    gravityOverride {z=-980}
    physicsBodyDefinitions [{
      boundBone {boneName=dyn_shirt_mid_l}
      boxExtents=<20 30 30>
      constraintSetup {angularLimitsMax=<2 20 10>, angularLimitsMin=<-2 -20 -10>, angularTarget=<1 0 0>}
      localJointOffset=<4 0 0>
     }]
   }
   n13 = Anim Dynamics - Bone: dyn_shirt_front_l!(ComponentPose=@n12, Alpha=1, GravityOverride:Vector="0,0,-980", ExternalForce="0,0,0") -> (Pose) {
    angularDampingOverride=1
    angularSpringConstant=11000
    bAngularSpring=true
    bOverrideAngularDamping=true
    bUseGravityOverride=true
    gravityScale=0.35
    simulationSpace=RootRelative
    alphaBoolBlend {blendOption=Linear}
    boundBone {boneName=dyn_shirt_front_l}
    componentAppliedLinearAccClamp=<9500 9500 9500>
    componentLinearAccScale=<1 1 1>
    componentLinearVelScale=<1 1 1>
    gravityOverride {z=-980}
    physicsBodyDefinitions [{
      boundBone {boneName=dyn_shirt_front_l}
      boxExtents=<20 30 30>
      constraintSetup {angularLimitsMax=<3 20 15>, angularLimitsMin=<-3 -20 -15>, angularTarget=<1 0 0>}
      localJointOffset=<4 0 0>
     }]
   }
   n14 = Anim Dynamics - Bone: dyn_strap_back_l_02!(ComponentPose=@n13, bAlphaBoolEnabled:bool=@n1, GravityOverride:Vector="0,0,-980", ExternalForce="0,0,0") -> (Pose) {
    alphaInputType=Bool
    angularDampingOverride=1
    angularSpringConstant=11000
    bAngularSpring=true
    bOverrideAngularDamping=true
    bUseGravityOverride=true
    gravityScale=0.35
    simulationSpace=RootRelative
    alphaBoolBlend {blendInTime=0.5, blendOption=Sinusoidal, blendOutTime=1}
    boundBone {boneName=dyn_strap_back_l_02}
    componentAppliedLinearAccClamp=<9500 9500 9500>
    componentLinearAccScale=<1 1 1>
    componentLinearVelScale=<1 1 1>
    gravityOverride {z=-980}
    physicsBodyDefinitions [{
      boundBone {boneName=dyn_strap_back_l_02}
      boxExtents=<20 30 30>
      constraintSetup {angularLimitsMax=<2 10 25>, angularLimitsMin=<-2 -40 -25>, angularTarget=<1 0 0>}
      localJointOffset=<4 0 0>
     }]
   }
   n15 = Anim Dynamics - Bone: dyn_strap_back_l_01!(ComponentPose=@n14, bAlphaBoolEnabled:bool=@n1, GravityOverride:Vector="0,0,-980", ExternalForce="0,0,0") -> (Pose) {
    alphaInputType=Bool
    angularDampingOverride=1
    angularSpringConstant=11000
    bAngularSpring=true
    bOverrideAngularDamping=true
    bUseGravityOverride=true
    gravityScale=0.35
    simulationSpace=RootRelative
    alphaBoolBlend {blendInTime=0.5, blendOption=Sinusoidal, blendOutTime=1}
    boundBone {boneName=dyn_strap_back_l_01}
    componentAppliedLinearAccClamp=<9500 9500 9500>
    componentLinearAccScale=<1 1 1>
    componentLinearVelScale=<1 1 1>
    gravityOverride {z=-980}
    physicsBodyDefinitions [{
      boundBone {boneName=dyn_strap_back_l_01}
      boxExtents=<20 30 30>
      constraintSetup {angularLimitsMax=<1 0 15>, angularLimitsMin=<-1 -30 -15>, angularTarget=<1 0 0>}
      localJointOffset=<4 0 0>
     }]
   }
   n16 = Anim Dynamics - Bone: dyn_strap_back_r_01!(ComponentPose=@n15, bAlphaBoolEnabled:bool=@n1, GravityOverride:Vector="0,0,-980", ExternalForce="0,0,0") -> (Pose) {
    alphaInputType=Bool
    angularDampingOverride=1
    angularSpringConstant=11000
    bAngularSpring=true
    bOverrideAngularDamping=true
    bUseGravityOverride=true
    gravityScale=0.35
    simulationSpace=RootRelative
    alphaBoolBlend {blendInTime=0.5, blendOption=Sinusoidal, blendOutTime=1}
    boundBone {boneName=dyn_strap_back_r_01}
    componentAppliedLinearAccClamp=<9500 9500 9500>
    componentLinearAccScale=<1 1 1>
    componentLinearVelScale=<1 1 1>
    gravityOverride {z=-980}
    physicsBodyDefinitions [{
      boundBone {boneName=dyn_strap_back_r_01}
      boxExtents=<20 30 30>
      constraintSetup {angularLimitsMax=<1 0 15>, angularLimitsMin=<-1 -30 -15>, angularTarget=<1 0 0>}
      localJointOffset=<4 0 0>
     }]
   }
   n17 = Anim Dynamics - Bone: dyn_strap_back_r_02!(ComponentPose=@n16, bAlphaBoolEnabled:bool=@n1, GravityOverride:Vector="0,0,-980", ExternalForce="0,0,0") -> (Pose) {
    alphaInputType=Bool
    angularDampingOverride=1
    angularSpringConstant=11000
    bAngularSpring=true
    bOverrideAngularDamping=true
    bUseGravityOverride=true
    gravityScale=0.35
    simulationSpace=RootRelative
    alphaBoolBlend {blendInTime=0.5, blendOption=Sinusoidal, blendOutTime=1}
    boundBone {boneName=dyn_strap_back_r_02}
    componentAppliedLinearAccClamp=<9500 9500 9500>
    componentLinearAccScale=<1 1 1>
    componentLinearVelScale=<1 1 1>
    gravityOverride {z=-980}
    physicsBodyDefinitions [{
      boundBone {boneName=dyn_strap_back_r_02}
      boxExtents=<20 30 30>
      constraintSetup {angularLimitsMax=<2 10 25>, angularLimitsMin=<-2 -40 -25>, angularTarget=<1 0 0>}
      localJointOffset=<4 0 0>
     }]
   }
   n18 = Anim Dynamics - Bone: dyn_daggers_back!(ComponentPose=@n17, bAlphaBoolEnabled:bool=@n1, GravityOverride:Vector="0,0,-980", ExternalForce="0,0,0") -> (Pose) {
    alphaInputType=Bool
    angularDampingOverride=1
    angularSpringConstant=11000
    bAngularSpring=true
    bOverrideAngularDamping=true
    bUseGravityOverride=true
    gravityScale=0.35
    simulationSpace=RootRelative
    alphaBoolBlend {blendInTime=0.5, blendOption=Sinusoidal, blendOutTime=1}
    boundBone {boneName=dyn_daggers_back}
    componentAppliedLinearAccClamp=<9500 9500 9500>
    componentLinearAccScale=<0.4 0.4 0.4>
    componentLinearVelScale=<0.4 0.4 0.4>
    gravityOverride {z=-980}
    physicsBodyDefinitions [{
      boundBone {boneName=dyn_daggers_back}
      boxExtents=<20 30 30>
      constraintSetup {angularLimitsMax=<3 15 3>, angularLimitsMin=<-3 -15 -3>, angularTarget=<1 0 0>}
      localJointOffset=<4 0 0>
     }]
   }
   n19 = "Transform (Modify) Bone - Bone: base_skirt_mid"(ComponentPose=@n18, Alpha=1, Translation="0,0,0", Rotation="10,0,0", Scale="1,1,1") {
    rotationMode=BMM_Additive
    rotationSpace=BCS_BoneSpace
    translationMode=BMM_Additive
    translationSpace=BCS_BoneSpace
    alphaBoolBlend {blendOption=Linear}
    boneToModify {boneName=base_skirt_mid}
    rotation {pitch=10}
   }
   n20 = Control Rig!(Source=Component To Local(ComponentPose=@n19), Alpha, Targeting:bool=Get bIsTargeting(), Crouching:bool=Get bIsCrouching()) -> (Pose) {
    alphaBoolBlend {blendOption=Linear}
    controlRigAssetReference {blueprintRigClass=&/M_MED_CyborgWarrior_CtrlRig.M_MED_CyborgWarrior_CtrlRig_C}
   }
   n21 = "Transform (Modify) Bone - Bone: base_skirt_back"(ComponentPose=Local To Component(LocalPose=@n20), Alpha=1, Translation="-5,-8,0", Rotation="0,0,0", Scale="1,1,1") {
    rotationMode=BMM_Additive
    rotationSpace=BCS_BoneSpace
    translationMode=BMM_Additive
    translationSpace=BCS_BoneSpace
    alphaBoolBlend {blendInTime=0.5, blendOption=QuadraticInOut, blendOutTime=1}
    boneToModify {boneName=base_skirt_back}
    translation=<-5 -8>
   }
   n22 = "Transform (Modify) Bone - Bone: base_skirt_front"!(ComponentPose=@n21, bAlphaBoolEnabled:bool=True, Translation="0,0,0", Rotation="-40,0,0", Scale="1,1,1") -> (Pose) {
    alphaInputType=Bool
    rotationMode=BMM_Additive
    rotationSpace=BCS_BoneSpace
    translationMode=BMM_Additive
    translationSpace=BCS_BoneSpace
    alphaBoolBlend {blendInTime=0.5, blendOption=QuadraticInOut, blendOutTime=1}
    boneToModify {boneName=base_skirt_front}
    rotation {pitch=-40}
   }
   n23 = "Transform (Modify) Bone - Bone: base_skirt_midfront"!(ComponentPose=@n22, bAlphaBoolEnabled:bool=True, Translation="0,0,0", Rotation="0,10,0", Scale="1,1,1") -> (Pose) {
    alphaInputType=Bool
    rotationMode=BMM_Additive
    rotationSpace=BCS_BoneSpace
    translationMode=BMM_Additive
    translationSpace=BCS_BoneSpace
    alphaBoolBlend {blendInTime=0.5, blendOption=QuadraticInOut, blendOutTime=1}
    boneToModify {boneName=base_skirt_midfront}
    rotation {yaw=10}
   }
   n24 = "Transform (Modify) Bone - Bone: base_skirt_mid"!(ComponentPose=@n23, bAlphaBoolEnabled:bool=True, Translation="0,0,0", Rotation="20,10,0", Scale="1,1,1") -> (Pose) {
    alphaInputType=Bool
    rotationMode=BMM_Additive
    rotationSpace=BCS_BoneSpace
    translationMode=BMM_Additive
    translationSpace=BCS_BoneSpace
    alphaBoolBlend {blendInTime=0.5, blendOption=QuadraticInOut, blendOutTime=1}
    boneToModify {boneName=base_skirt_mid}
    rotation {pitch=20, yaw=10}
   }
   n25 = Get FrontEndWithoutEmote() -> (FrontEndWithoutEmote:bool)
   n26 = RigidBody!(ComponentPose=@n24, bAlphaBoolEnabled:bool=NOT Boolean(A=@n25), OverrideWorldGravity:Vector=Get GravityOverride(), ExternalForce) -> (Pose) {
    alphaInputType=Bool
    bClampLinearTranslationLimitToRefPose=true
    bOverrideWorldGravity=true
    overridePhysicsAsset=&/M_MED_CyborgWarrior_Physics
    simulationSpace=BaseBoneSpace
    alphaBoolBlend {blendInTime=0.5, blendOption=QuadraticInOut, blendOutTime=1}
    baseBoneRef {boneName=pelvis}
    componentAppliedLinearAccClamp=<0 0 0>
    simSpaceSettings {
     maxAngularAcceleration=200
     maxAngularVelocity=75
     maxLinearAcceleration=800
     maxLinearVelocity=550
     worldAlpha=1
     externalLinearDragV=<0.25 0.5>
    }
   }
   n27 = RigidBody!(
    ComponentPose=@n26
    bAlphaBoolEnabled=@n25
    OverrideWorldGravity=Get GravityOverride()
    ExternalForce
    SimSpaceSettings:Sim Space Settings Structure=Get Sim Space Settings() -> (Sim Space Settings:Sim Space Settings Structure)
   ) -> (Pose) {
    alphaInputType=Bool
    bClampLinearTranslationLimitToRefPose=true
    bOverrideWorldGravity=true
    overridePhysicsAsset=&M_MED_CyborgWarrior_Physics
    simulationSpace=BaseBoneSpace
    alphaBoolBlend {blendInTime=0.5, blendOption=QuadraticInOut, blendOutTime=1}
    baseBoneRef {boneName=thigh_r}
    componentAppliedLinearAccClamp=<0 0 0>
    simSpaceSettings {
     maxAngularAcceleration=200
     maxAngularVelocity=75
     maxLinearAcceleration=800
     maxLinearVelocity=550
     worldAlpha=1
     externalLinearDragV=<0.25 0.5>
    }
   }
   n28 = "Transform (Modify) Bone - Bone: base_skirt_front"!(ComponentPose=@n27, Alpha:float=1, Translation="0,0,0", Rotation="0,0,-30", Scale="1,1,1") -> (Pose) {
    rotationMode=BMM_Additive
    rotationSpace=BCS_BoneSpace
    translationMode=BMM_Additive
    translationSpace=BCS_BoneSpace
    alphaBoolBlend {blendInTime=0.5, blendOption=QuadraticInOut, blendOutTime=1}
    boneToModify {boneName=base_skirt_front}
    rotation {roll=-30}
   }
   n29 = "Transform (Modify) Bone - Bone: base_skirt_midfront"!(ComponentPose=@n28, Alpha:float=1, Translation="0,0,0", Rotation="60,0,-10", Scale="1,1,1") -> (Pose) {
    rotationMode=BMM_Additive
    rotationSpace=BCS_BoneSpace
    translationMode=BMM_Additive
    translationSpace=BCS_BoneSpace
    alphaBoolBlend {blendInTime=0.5, blendOption=QuadraticInOut, blendOutTime=1}
    boneToModify {boneName=base_skirt_midfront}
    rotation {pitch=60, roll=-10}
   }
   n30 = "Transform (Modify) Bone - Bone: base_skirt_mid"!(ComponentPose=@n29, Alpha:float=1, Translation="0,0,0", Rotation="65,0,-5", Scale="1,1,1") -> (Pose) {
    rotationMode=BMM_Additive
    rotationSpace=BCS_BoneSpace
    translationMode=BMM_Additive
    translationSpace=BCS_BoneSpace
    alphaBoolBlend {blendInTime=0.5, blendOption=QuadraticInOut, blendOutTime=1}
    boneToModify {boneName=base_skirt_mid}
    rotation {pitch=65, roll=-5}
   }
   n31 = "Transform (Modify) Bone - Bone: base_skirt_midback"(ComponentPose=@n30, Alpha=1, Translation="0,0,0", Rotation="60,0,0", Scale="1,1,1") {
    rotationMode=BMM_Additive
    rotationSpace=BCS_BoneSpace
    translationMode=BMM_Additive
    translationSpace=BCS_BoneSpace
    alphaBoolBlend {blendInTime=0.5, blendOption=QuadraticInOut, blendOutTime=1}
    boneToModify {boneName=base_skirt_midback}
    rotation {pitch=60}
   }
   n32 = "Transform (Modify) Bone - Bone: base_skirt_back"(ComponentPose=@n31, Alpha=1, Translation="0,0,0", Rotation="45,0,50", Scale="1,1,1") {
    rotationMode=BMM_Additive
    rotationSpace=BCS_BoneSpace
    translationMode=BMM_Additive
    translationSpace=BCS_BoneSpace
    alphaBoolBlend {blendInTime=0.5, blendOption=QuadraticInOut, blendOutTime=1}
    boneToModify {boneName=base_skirt_back}
    rotation {pitch=45, roll=50}
   }
   n33 = "Transform (Modify) Bone - Bone: base_skirt_midfront"(ComponentPose=@n32, Alpha=1, Translation="0,0,0", Rotation="5,0,0", Scale="1,1,1") {
    rotationMode=BMM_Additive
    rotationSpace=BCS_BoneSpace
    translationMode=BMM_Additive
    translationSpace=BCS_BoneSpace
    alphaBoolBlend {blendInTime=0.5, blendOption=QuadraticInOut, blendOutTime=1}
    boneToModify {boneName=base_skirt_midfront}
    rotation {pitch=5}
   }
   n34 = "Transform (Modify) Bone - Bone: base_skirt_mid"(ComponentPose=@n33, Alpha=1, Translation="0,0,0", Rotation="25,0,0", Scale="1,1,1") {
    rotationMode=BMM_Additive
    rotationSpace=BCS_BoneSpace
    translationMode=BMM_Additive
    translationSpace=BCS_BoneSpace
    alphaBoolBlend {blendInTime=0.5, blendOption=QuadraticInOut, blendOutTime=1}
    boneToModify {boneName=base_skirt_mid}
    rotation {pitch=25}
   }
   n35 = "Transform (Modify) Bone - Bone: base_skirt_midback"(ComponentPose=@n34, Alpha=1, Translation="0,0,0", Rotation="20,0,0", Scale="1,1,1") {
    rotationMode=BMM_Additive
    rotationSpace=BCS_BoneSpace
    translationMode=BMM_Additive
    translationSpace=BCS_BoneSpace
    alphaBoolBlend {blendInTime=0.5, blendOption=QuadraticInOut, blendOutTime=1}
    boneToModify {boneName=base_skirt_midback}
    rotation {pitch=20}
   }
   n36 = "Transform (Modify) Bone - Bone: base_skirt_back"(ComponentPose=@n35, Alpha=1, Translation="0,0,0", Rotation="5,0,15", Scale="1,1,1") {
    rotationMode=BMM_Additive
    rotationSpace=BCS_BoneSpace
    translationMode=BMM_Additive
    translationSpace=BCS_BoneSpace
    alphaBoolBlend {blendInTime=0.5, blendOption=QuadraticInOut, blendOutTime=1}
    boneToModify {boneName=base_skirt_back}
    rotation {pitch=5, roll=15}
   }
   n37 = "Transform (Modify) Bone - Bone: base_skirt_back"!(ComponentPose=@n36, bAlphaBoolEnabled:bool=True, Translation="0,0,0", Rotation="15,0,0", Scale="1,1,1") -> (Pose) {
    alphaInputType=Bool
    rotationMode=BMM_Additive
    rotationSpace=BCS_BoneSpace
    translationMode=BMM_Additive
    translationSpace=BCS_BoneSpace
    alphaBoolBlend {blendInTime=0.5, blendOption=QuadraticInOut, blendOutTime=1}
    boneToModify {boneName=base_skirt_back}
    rotation {pitch=15}
   }
   n38 = "Transform (Modify) Bone - Bone: base_skirt_midback"!(ComponentPose=@n37, bAlphaBoolEnabled:bool=True, Translation="0,0,0", Rotation="0,10,0", Scale="1,1,1") -> (Pose) {
    alphaInputType=Bool
    rotationMode=BMM_Additive
    rotationSpace=BCS_BoneSpace
    translationMode=BMM_Additive
    translationSpace=BCS_BoneSpace
    alphaBoolBlend {blendInTime=0.5, blendOption=QuadraticInOut, blendOutTime=1}
    boneToModify {boneName=base_skirt_midback}
    rotation {yaw=10}
   }
   n39 = "Transform (Modify) Bone - Bone: base_skirt_front"!(ComponentPose=@n38, bAlphaBoolEnabled:bool=True, Translation="0,2,0", Rotation="20,15,0", Scale="1,1,1") -> (Pose) {
    alphaInputType=Bool
    rotationMode=BMM_Additive
    rotationSpace=BCS_BoneSpace
    translationMode=BMM_Additive
    translationSpace=BCS_BoneSpace
    alphaBoolBlend {blendInTime=0.1, blendOption=Sinusoidal, blendOutTime=0.8}
    boneToModify {boneName=base_skirt_front}
    rotation {pitch=20, yaw=15}
    translation {y=2}
   }
   n40 = "Transform (Modify) Bone - Bone: base_skirt_mid"!(ComponentPose=@n39, bAlphaBoolEnabled:bool=True, Translation="1.5,0,0", Rotation="0,0,0", Scale="1,1,1") -> (Pose) {
    alphaInputType=Bool
    rotationMode=BMM_Additive
    rotationSpace=BCS_BoneSpace
    translationMode=BMM_Additive
    translationSpace=BCS_BoneSpace
    alphaBoolBlend {blendInTime=0.5, blendOption=QuadraticInOut, blendOutTime=1}
    boneToModify {boneName=base_skirt_mid}
    translation {x=1.5}
   }
   n41 = "Transform (Modify) Bone - Bone: base_skirt_front"(ComponentPose=@n40, bAlphaBoolEnabled=True, Translation="6,-1,0", Rotation="0,-15,15", Scale="1,1,1") {
    alphaInputType=Bool
    rotationMode=BMM_Additive
    rotationSpace=BCS_BoneSpace
    translationMode=BMM_Additive
    translationSpace=BCS_BoneSpace
    alphaBoolBlend {blendInTime=0.5, blendOption=QuadraticInOut, blendOutTime=1}
    boneToModify {boneName=base_skirt_front}
    rotation {roll=15, yaw=-15}
    translation=<6 -1>
   }
   n42 = "Transform (Modify) Bone - Bone: base_skirt_back"(ComponentPose=@n41, bAlphaBoolEnabled=True, Translation="0,0,0", Rotation="10,-5,0", Scale="1,1,1") {
    alphaInputType=Bool
    rotationMode=BMM_Additive
    rotationSpace=BCS_BoneSpace
    translationMode=BMM_Additive
    translationSpace=BCS_BoneSpace
    alphaBoolBlend {blendInTime=0.5, blendOption=QuadraticInOut, blendOutTime=1}
    boneToModify {boneName=base_skirt_back}
    rotation {pitch=10, yaw=-5}
   }
   n43 = "Transform (Modify) Bone - Bone: base_skirt_front"(ComponentPose=@n42, bAlphaBoolEnabled=True, Translation="0,0,0", Rotation="8,8,0", Scale="1,1,1") {
    alphaInputType=Bool
    rotationMode=BMM_Additive
    rotationSpace=BCS_BoneSpace
    translationMode=BMM_Additive
    translationSpace=BCS_BoneSpace
    alphaBoolBlend {blendInTime=0.5, blendOption=QuadraticInOut, blendOutTime=1}
    boneToModify {boneName=base_skirt_front}
    rotation {pitch=8, yaw=8}
   }
   n44 = "Transform (Modify) Bone - Bone: base_skirt_front"(ComponentPose=@n43, bAlphaBoolEnabled=True, Translation="0,0,0", Rotation="10,10,0", Scale="1,1,1") {
    alphaInputType=Bool
    rotationMode=BMM_Additive
    rotationSpace=BCS_BoneSpace
    translationMode=BMM_Additive
    translationSpace=BCS_BoneSpace
    alphaBoolBlend {blendInTime=0.5, blendOption=QuadraticInOut, blendOutTime=1}
    boneToModify {boneName=base_skirt_front}
    rotation {pitch=10, yaw=10}
   }
   n45 = "Transform (Modify) Bone - Bone: base_skirt_front"(ComponentPose=@n44, bAlphaBoolEnabled=True, Translation="0,1,0", Rotation="10,10,0", Scale="1,1,1") {
    alphaInputType=Bool
    rotationMode=BMM_Additive
    rotationSpace=BCS_BoneSpace
    translationMode=BMM_Additive
    translationSpace=BCS_BoneSpace
    alphaBoolBlend {blendInTime=0.5, blendOption=QuadraticInOut, blendOutTime=1}
    boneToModify {boneName=base_skirt_front}
    rotation {pitch=10, yaw=10}
    translation {y=1}
   }
   n46 = "Transform (Modify) Bone - Bone: base_skirt_midfront"!(ComponentPose=@n45, bAlphaBoolEnabled:bool=True, Translation="0,0,0", Rotation="10,0,0", Scale="1,1,1") -> (Pose) {
    alphaInputType=Bool
    rotationMode=BMM_Additive
    rotationSpace=BCS_BoneSpace
    translationMode=BMM_Additive
    translationSpace=BCS_BoneSpace
    alphaBoolBlend {blendInTime=0.5, blendOption=QuadraticInOut, blendOutTime=1}
    boneToModify {boneName=base_skirt_midfront}
    rotation {pitch=10}
   }
   n47 = "Transform (Modify) Bone - Bone: base_skirt_front"(ComponentPose=@n46, bAlphaBoolEnabled=True, Translation="0,0,0", Rotation="5,0,0", Scale="1,1,1") {
    alphaInputType=Bool
    rotationMode=BMM_Additive
    rotationSpace=BCS_BoneSpace
    translationMode=BMM_Additive
    translationSpace=BCS_BoneSpace
    alphaBoolBlend {blendInTime=0.5, blendOption=QuadraticInOut, blendOutTime=1}
    boneToModify {boneName=base_skirt_front}
    rotation {pitch=5}
   }
   n48 = "Transform (Modify) Bone - Bone: base_skirt_mid"(ComponentPose=@n47, bAlphaBoolEnabled=True, Translation="0,0,0", Rotation="-5,-15,0", Scale="1,1,1") {
    alphaInputType=Bool
    rotationMode=BMM_Additive
    rotationSpace=BCS_BoneSpace
    translationMode=BMM_Additive
    translationSpace=BCS_BoneSpace
    alphaBoolBlend {blendInTime=0.5, blendOption=QuadraticInOut, blendOutTime=1}
    boneToModify {boneName=base_skirt_mid}
    rotation {pitch=-5, yaw=-15}
   }
   n49 = "Transform (Modify) Bone - Bone: base_skirt_midback"(ComponentPose=@n48, bAlphaBoolEnabled=True, Translation="0,0,0", Rotation="-10,5,0", Scale="1,1,1") {
    alphaInputType=Bool
    rotationMode=BMM_Additive
    rotationSpace=BCS_BoneSpace
    translationMode=BMM_Additive
    translationSpace=BCS_BoneSpace
    alphaBoolBlend {blendInTime=0.5, blendOption=QuadraticInOut, blendOutTime=1}
    boneToModify {boneName=base_skirt_midback}
    rotation {pitch=-10, yaw=5}
   }
   n50 = "Transform (Modify) Bone - Bone: base_skirt_back"(ComponentPose=@n49, bAlphaBoolEnabled=True, Translation="0,0,0", Rotation="2,0,0", Scale="1,1,1") {
    alphaInputType=Bool
    rotationMode=BMM_Additive
    rotationSpace=BCS_BoneSpace
    translationMode=BMM_Additive
    translationSpace=BCS_BoneSpace
    alphaBoolBlend {blendInTime=0.5, blendOption=QuadraticInOut, blendOutTime=1}
    boneToModify {boneName=base_skirt_back}
    rotation {pitch=2}
   }
   n51 = "Transform (Modify) Bone - Bone: base_neck_mid_l"!(ComponentPose=@n50, bAlphaBoolEnabled:bool=True, Translation="0,0,5", Rotation="0,0,0", Scale="1,1,1") -> (Pose) {
    alphaInputType=Bool
    rotationMode=BMM_Additive
    rotationSpace=BCS_BoneSpace
    translationMode=BMM_Additive
    translationSpace=BCS_BoneSpace
    alphaBoolBlend {blendInTime=0.5, blendOption=QuadraticInOut, blendOutTime=1}
    boneToModify {boneName=base_neck_mid_l}
    translation {z=5}
   }
   n52 = "Transform (Modify) Bone - Bone: base_neck_frontmid_l"!(ComponentPose=@n51, bAlphaBoolEnabled:bool=True, Translation="0,6,2", Rotation="0,0,0", Scale="1,1,1") -> (Pose) {
    alphaInputType=Bool
    rotationMode=BMM_Additive
    rotationSpace=BCS_BoneSpace
    translationMode=BMM_Additive
    translationSpace=BCS_BoneSpace
    alphaBoolBlend {blendInTime=0.5, blendOption=QuadraticInOut, blendOutTime=1}
    boneToModify {boneName=base_neck_frontmid_l}
    translation {y=6, z=2}
   }
   n53 = "Transform (Modify) Bone - Bone: base_neck_frontmid_r"!(ComponentPose=@n52, bAlphaBoolEnabled:bool=True, Translation="0,6,2", Rotation="0,0,0", Scale="1,1,1") -> (Pose) {
    alphaInputType=Bool
    rotationMode=BMM_Additive
    rotationSpace=BCS_BoneSpace
    translationMode=BMM_Additive
    translationSpace=BCS_BoneSpace
    alphaBoolBlend {blendInTime=0.5, blendOption=QuadraticInOut, blendOutTime=1}
    boneToModify {boneName=base_neck_frontmid_r}
    translation {y=6, z=2}
   }
   n54 = "Transform (Modify) Bone - Bone: base_neck_mid_r"!(ComponentPose=@n53, bAlphaBoolEnabled:bool=True, Translation="0,0,5", Rotation="0,0,0", Scale="1,1,1") -> (Pose) {
    alphaInputType=Bool
    rotationMode=BMM_Additive
    rotationSpace=BCS_BoneSpace
    translationMode=BMM_Additive
    translationSpace=BCS_BoneSpace
    alphaBoolBlend {blendInTime=0.5, blendOption=QuadraticInOut, blendOutTime=1}
    boneToModify {boneName=base_neck_mid_r}
    translation {z=5}
   }
   n55 = "Transform (Modify) Bone - Bone: dyn_skirt_front_02"!(ComponentPose=@n54, bAlphaBoolEnabled:bool=True, Translation="0,0,0", Rotation="5,0,0", Scale="1,1,1") -> (Pose) {
    alphaInputType=Bool
    rotationMode=BMM_Additive
    rotationSpace=BCS_BoneSpace
    translationMode=BMM_Additive
    translationSpace=BCS_BoneSpace
    alphaBoolBlend {blendInTime=0.5, blendOption=QuadraticInOut, blendOutTime=1}
    boneToModify {boneName=dyn_skirt_front_02}
    rotation {pitch=5}
   }
   n56 = "Transform (Modify) Bone - Bone: base_skirt_front"(ComponentPose=@n55, bAlphaBoolEnabled=True, Translation="0,0,4", Rotation="-15,-10,0", Scale="1,1,1") {
    alphaInputType=Bool
    rotationMode=BMM_Additive
    rotationSpace=BCS_BoneSpace
    translationMode=BMM_Additive
    translationSpace=BCS_BoneSpace
    alphaBoolBlend {blendInTime=0.5, blendOption=QuadraticInOut, blendOutTime=1}
    boneToModify {boneName=base_skirt_front}
    rotation {pitch=-15, yaw=-10}
    translation {z=4}
   }
   n57 = "Transform (Modify) Bone - Bone: base_skirt_back"(ComponentPose=@n56, bAlphaBoolEnabled=True, Translation="0,0,0", Rotation="-6,0,0", Scale="1,1,1") {
    alphaInputType=Bool
    rotationMode=BMM_Additive
    rotationSpace=BCS_BoneSpace
    translationMode=BMM_Additive
    translationSpace=BCS_BoneSpace
    alphaBoolBlend {blendInTime=0.5, blendOption=QuadraticInOut, blendOutTime=1}
    boneToModify {boneName=base_skirt_back}
    rotation {pitch=-6}
   }
   n58 = "Transform (Modify) Bone - Bone: base_skirt_mid"(ComponentPose=@n57, bAlphaBoolEnabled=True, Translation="0,0,0", Rotation="-5,0,0", Scale="1,1,1") {
    alphaInputType=Bool
    rotationMode=BMM_Additive
    rotationSpace=BCS_BoneSpace
    translationMode=BMM_Additive
    translationSpace=BCS_BoneSpace
    alphaBoolBlend {blendInTime=0.5, blendOption=QuadraticInOut, blendOutTime=1}
    boneToModify {boneName=base_skirt_mid}
    rotation {pitch=-5}
   }
   n59 = "Transform (Modify) Bone - Bone: base_skirt_midfront"(ComponentPose=@n58, bAlphaBoolEnabled=True, Translation="0,0,0", Rotation="-15,0,0", Scale="1,1,1") {
    alphaInputType=Bool
    rotationMode=BMM_Additive
    rotationSpace=BCS_BoneSpace
    translationMode=BMM_Additive
    translationSpace=BCS_BoneSpace
    alphaBoolBlend {blendInTime=0.5, blendOption=QuadraticInOut, blendOutTime=1}
    boneToModify {boneName=base_skirt_midfront}
    rotation {pitch=-15}
   }
   n60 = "Transform (Modify) Bone - Bone: base_skirt_front"(ComponentPose=@n59, bAlphaBoolEnabled=True, Translation="0,0,0", Rotation="-5,0,0", Scale="1,1,1") {
    alphaInputType=Bool
    rotationMode=BMM_Additive
    rotationSpace=BCS_BoneSpace
    translationMode=BMM_Additive
    translationSpace=BCS_BoneSpace
    alphaBoolBlend {blendInTime=0.5, blendOption=QuadraticInOut, blendOutTime=1}
    boneToModify {boneName=base_skirt_front}
    rotation {pitch=-5}
   }
   n61 = "Transform (Modify) Bone - Bone: attach_cape"(ComponentPose=@n60, Alpha=1, Translation="0,-1.5,0", Rotation="0,0,0", Scale="1,1,1") {
    rotationMode=BMM_Additive
    rotationSpace=BCS_BoneSpace
    translationMode=BMM_Additive
    translationSpace=BCS_BoneSpace
    alphaBoolBlend {blendOption=Linear}
    boneToModify {boneName=attach_cape}
    translation {y=-1.5}
   }
   n62 = "Transform (Modify) Bone - Bone: attach_backpack"(ComponentPose=@n61, Alpha=1, Translation="0,-1.5,0", Rotation="0,0,0", Scale="1,1,1") {
    rotationMode=BMM_Additive
    rotationSpace=BCS_BoneSpace
    translationMode=BMM_Additive
    translationSpace=BCS_BoneSpace
    alphaBoolBlend {blendOption=Linear}
    boneToModify {boneName=attach_backpack}
    translation {y=-1.5}
   }
   n63 = "Constraint - base_strap_front_l (PPPPPPPP)"!(
    ComponentPose=@n62
    Alpha=1
    ConstraintWeights_0=0.3095
    ConstraintWeights_1:float=0.1816
    ConstraintWeights_2:float=0.1623
    ConstraintWeights_3:float=0.1417
    ConstraintWeights_4:float=0.0707
    ConstraintWeights_5:float=0.0627
    ConstraintWeights_6:float=0.0619
    ConstraintWeights_7:float=0.0097
   ) -> (Pose) {
    alphaBoolBlend {blendOption=Linear}
    boneToModify {boneName=base_strap_front_l}
    constraintSetup [
     {targetBone {boneName=spine_03}}
     {targetBone {boneName=spine_02}}
     {targetBone {boneName=dyn_shirt_front_l}}
     {targetBone {boneName=spine_04}}
     {targetBone {boneName=spine_01}}
     {targetBone {boneName=deform_pec_l}}
     {targetBone {boneName=clavicle_l}}
     {targetBone {boneName=pelvis}}
    ]
    constraintWeights [1,1,1,1,1,1,1,1]
   }
   n64 = "Constraint - base_strap_skirtbag (PPPPPPPP)"!(
    ComponentPose=@n63
    Alpha=1
    ConstraintWeights_0=0.4259
    ConstraintWeights_1:float=0.2431
    ConstraintWeights_2:float=0.1206
    ConstraintWeights_3:float=0.0878
    ConstraintWeights_4:float=0.076
    ConstraintWeights_5:float=0.0369
    ConstraintWeights_6:float=0.0084
    ConstraintWeights_7:float=0.0014
   ) -> (Pose) {
    alphaBoolBlend {blendOption=Linear}
    boneToModify {boneName=base_strap_skirtbag}
    constraintSetup [
     {targetBone {boneName=dyn_skirt_mid_02}}
     {targetBone {boneName=dyn_skirt_midback_02}}
     {targetBone {boneName=dyn_skirt_mid_01}}
     {targetBone {boneName=dyn_skirt_midback_01}}
     {targetBone {boneName=dyn_skirt_midfront_02}}
     {targetBone {boneName=dyn_skirt_midfront_01}}
     {targetBone {boneName=dyn_skirt_mid_03}}
     {targetBone {boneName=dyn_skirt_midback_03}}
    ]
    constraintWeights [1,1,1,1,1,1,1,1]
   }
   Output Pose(Result=Component To Local(ComponentPose=@n64)) {layerGroup=None}
  ]}
 EventGraph {nodes [
   Event BlueprintUpdateAnimation()
   > n1 = Sequence!(execute) -> (then_0, then_1, then_2:Exec, then_3:Exec, then_4:Exec, then_5:Exec)
   n2 = Get bIsPlayingEmote()
   Set FrontEndWithoutEmote(execute=@n1.then_5, FrontEndWithoutEmote:bool=AND Boolean(A=Get bIsInFrontEnd(), B=NOT Boolean(A=@n2))) -> (Output_Get:bool)
   n3 = Gravity Override State Function DEPRECATED(
    execute=@n1.then_0
    CharacterPartAnimInstance=Self-Reference()
    GameTimeInSeconds=GetGameTimeInSeconds()
    GravityOverrideParams=Get Gravity Override Params()
    RBAN Params=Get RBAN Params()
   )
   > Set GravityOverride(GravityOverride=@n3)
   > Set Sim Space Settings(Sim Space Settings:Sim Space Settings Structure=Break RBANSetting Struct!(RBANSettingStruct:"RBANSetting Struct Structure (by ref)"=@n3.RBAN SettingsOut) -> (
     ComponentLinearAccScale_28_#1:Vector=~
     ComponentLinearVelScale_31_#2:Vector=~
     SimSpaceSettings_34_#3:Sim Space Settings Structure
    ) {is=title K2Node_BreakStruct}.SimSpaceSettings_34_#3) -> (Output_Get:Sim Space Settings Structure)
   n4 = Branch(execute=@n1.then_4, Condition=Get bIsSliding())
   > n5 = Set Left_Calf(Left_Calf:double=GetAlphaFromReferenceTransformDelta(SkeletalMeshComponent=GetPartSkeletalMeshComponent(), BoneName=calf_l, BaseName=thigh_l, InputRange=(X=0,Y=1), TwistAxis="1,0,0").RawOutput) -> (Output_Get:double)
   > n6 = Branch(Condition=NOT Boolean(A="float > float"(A=@n5, B=100)))
   > Set Knee_Slide(Knee_Slide:double=0) -> (Output_Get:double)
   > Set Baseball_Slide(Baseball_Slide:double=1) -> (Output_Get:double)
   Set Knee_Slide(execute=@n6.else, Knee_Slide=1)
   > Set Baseball_Slide(Baseball_Slide=0)
   Set Knee_Slide(execute=@n4.else, Knee_Slide=0)
   > Set Baseball_Slide(Baseball_Slide=0)
   n7 = Get bIsCrouching()
   n8 = Get bIsMoving2D()
   Set CrouchingMoving(execute=@n1.then_2, CrouchingMoving:bool=AND Boolean(A=@n7, B=@n8)) -> (Output_Get:bool)
   > Set CrouchingNotMove(CrouchingNotMove:bool=AND Boolean(A=@n7, B=NOT Boolean(A=@n8))) -> (Output_Get:bool)
   > Set MovingNotCrouching(MovingNotCrouching:bool=AND Boolean(A=@n8, B=NOT Boolean(A=@n7))) -> (Output_Get:bool)
   n9 = "integer >= integer"(A=Get CurrentLOD(), B=1)
   Set LOD1(execute=@n1.then_3, LOD1:bool=@n9) -> (Output_Get:bool)
   > Set FELOD1(FELOD1:bool=AND Boolean(A=@n9, B=Get bIsInFrontEnd())) -> (Output_Get:bool)
   n10 = NOT Boolean(A=Get bIsMoving2D())
   n11 = Get Skel Mesh()
   n12 = "float > float"(
    A=GetAlphaFromReferenceTransformDelta(SkeletalMeshComponent=@n11, BoneName=neck_01, BaseName=root, DeltaComponentType=EulerX, InputRange=(X=0,Y=10), TwistAxis="1,0,0")
    B=0.95
   )
   n13 = "float > float"(
    A=GetAlphaFromReferenceTransformDelta(SkeletalMeshComponent=@n11, BoneName=spine_01, BaseName=root, DeltaComponentType=EulerX, InputRange=(X=0,Y=-55), TwistAxis="1,0,0")
    B=0.5
   )
   Set RotOwnAxis(execute=@n1.then_1, RotOwnAxis:bool=OR Boolean(A=AND Boolean(A=@n13, B=@n10), B=AND Boolean(A=@n10, B=@n12))) -> (Output_Get:bool)
   Event OnCharacterCustomizationLoaded()
   n14 = CheckForBackbling()
   > Set IsWearingBackpackOrCape(IsWearingBackpackOrCape:bool=OR Boolean!(A=@n14, B=@n14.bIsBackblingACape, C:bool=@n14.bIsBackblingLong) -> (ReturnValue)) -> (Output_Get:bool)
   Event BlueprintInitializeAnimation()
   > Set Skel Mesh(Skel Mesh=GetOwnerSkeletalMeshComponent())
   Event OnBeginCrouching()
   > Set StartCrouching(StartCrouching:bool=true) -> (Output_Get:bool)
   > Delay(Duration=0.3)
   > Set StartCrouching(StartCrouching=false)
   Event OnEndCrouching()
   > Set StartCrouching(StartCrouching=true)
   > Delay(Duration=0.3)
   > Set StartCrouching(StartCrouching=false)
  ]}
]
variables [
 Baseball_Slide {type=number}
 CrouchingMoving {type=boolean}
 CrouchingNotMove {type=boolean}
 FELOD1 {type=boolean}
 FrontEndWithoutEmote {type=boolean}
 Gravity Override Params {type=GravityOverrideParamsStruct, default {skydive_GoingUp {gravityOverride=<0 0 -980>}}}
 GravityOverride {type=Vector}
 IsWearingBackpackOrCape {type=boolean}
 Knee_Slide {type=number}
 Left_Calf {type=number}
 LOD1 {type=boolean}
 MovingNotCrouching {type=boolean}
 RBAN Params {
  type=RBANParamsStruct
  default {
   dBNO {simSpaceSettings {
     maxAngularAcceleration=1000
     maxAngularVelocity=3
     maxLinearAcceleration=300
     maxLinearVelocity=700
     velocityScaleZ=0
     externalLinearDragV=<0 0.5 0>
    }}
   dBNO_Carried {simSpaceSettings {
     maxAngularAcceleration=1000
     maxAngularVelocity=3
     maxLinearAcceleration=300
     maxLinearVelocity=700
     velocityScaleZ=0
     externalLinearDragV=<0 0.5 0>
    }}
   emoteOrMelee {simSpaceSettings {
     maxAngularAcceleration=1000
     maxAngularVelocity=3
     maxLinearAcceleration=300
     maxLinearVelocity=700
     velocityScaleZ=0
     externalLinearDragV=<0 0.5 0>
    }}
   front_End {simSpaceSettings {
     maxAngularAcceleration=1000
     maxAngularVelocity=3
     maxLinearAcceleration=300
     maxLinearVelocity=700
     velocityScaleZ=0
     externalLinearDragV=<0 0.5 0>
    }}
   onGround_Crouch_Moving {simSpaceSettings {
     maxAngularAcceleration=1000
     maxAngularVelocity=3
     maxLinearAcceleration=300
     maxLinearVelocity=700
     velocityScaleZ=0
     externalLinearDragV=<0 0.5 0>
    }}
   onGround_Falling {simSpaceSettings {
     maxAngularAcceleration=1000
     maxAngularVelocity=3
     maxLinearAcceleration=300
     maxLinearVelocity=700
     velocityScaleZ=0
     externalLinearDragV=<0 0.5 0>
    }}
   onGround_Moving {simSpaceSettings {
     maxAngularAcceleration=1000
     maxAngularVelocity=3
     maxLinearAcceleration=300
     maxLinearVelocity=700
     velocityScaleZ=0
     externalLinearDragV=<0 0.5 0>
    }}
   onGround_Standing {simSpaceSettings {
     maxAngularAcceleration=1000
     maxAngularVelocity=3
     maxLinearAcceleration=300
     maxLinearVelocity=700
     velocityScaleZ=0
     externalLinearDragV=<0 0.5 0>
    }}
   rail_Grinding {simSpaceSettings {
     maxAngularAcceleration=1000
     maxAngularVelocity=3
     maxLinearAcceleration=300
     maxLinearVelocity=700
     velocityScaleZ=0
     worldAlpha=0.1
     externalLinearDragV=<0 0.5 0>
    }}
   riding_Moving {simSpaceSettings {
     maxAngularAcceleration=1000
     maxAngularVelocity=3
     maxLinearAcceleration=300
     maxLinearVelocity=700
     velocityScaleZ=0
     externalLinearDragV=<0 0.5 0>
    }}
   riding_Moving_Targeting {simSpaceSettings {
     maxAngularAcceleration=1000
     maxAngularVelocity=3
     maxLinearAcceleration=300
     maxLinearVelocity=700
     velocityScaleZ=0
     externalLinearDragV=<0 0.5 0>
    }}
   riding_Standing {simSpaceSettings {
     maxAngularAcceleration=1000
     maxAngularVelocity=3
     maxLinearAcceleration=300
     maxLinearVelocity=700
     velocityScaleZ=0
     externalLinearDragV=<0 0.5 0>
    }}
   skydive_DivingDown {simSpaceSettings {
     maxAngularAcceleration=1000
     maxAngularVelocity=3
     maxLinearAcceleration=300
     maxLinearVelocity=700
     velocityScaleZ=0
     externalLinearDragV=<0 0.5 0>
    }}
   skydive_FloatingDown {
    componentLinearAccScale=<0 0.5 0>
    componentLinearVelScale=<0 0.5 0>
    simSpaceSettings {
     maxAngularAcceleration=1000
     maxAngularVelocity=3
     maxLinearAcceleration=300
     maxLinearVelocity=700
     velocityScaleZ=0
     externalLinearDragV=<0 0.5 0>
    }
   }
   skydive_GoingUp {simSpaceSettings {
     maxAngularAcceleration=1000
     maxAngularVelocity=3
     maxLinearAcceleration=300
     maxLinearVelocity=700
     velocityScaleZ=0
     externalLinearDragV=<0 0.5 0>
    }}
   skydive_Parachuting {
    componentLinearAccScale=<0 0.1 0.3>
    componentLinearVelScale=<0 0.1 0.3>
    simSpaceSettings {
     maxAngularAcceleration=1000
     maxAngularVelocity=3
     maxLinearAcceleration=300
     maxLinearVelocity=700
     velocityScaleZ=0
     externalLinearDragV=<0 0.5 0>
    }
   }
   skydive_Surfing {simSpaceSettings {
     maxAngularAcceleration=1000
     maxAngularVelocity=3
     maxLinearAcceleration=300
     maxLinearVelocity=700
     velocityScaleZ=0
     externalLinearDragV=<0 0.5 0>
    }}
   swimming {simSpaceSettings {
     maxAngularAcceleration=1000
     maxAngularVelocity=3
     maxLinearAcceleration=300
     maxLinearVelocity=700
     velocityScaleZ=0
     externalLinearDragV=<0 0.5 0>
    }}
   tactical_Sprint {simSpaceSettings {
     maxAngularAcceleration=1000
     maxAngularVelocity=3
     maxLinearAcceleration=300
     maxLinearVelocity=700
     velocityScaleZ=0
     externalLinearDragV=<0 0.5 0>
    }}
   ziplining {simSpaceSettings {
     maxAngularAcceleration=1000
     maxAngularVelocity=3
     maxLinearAcceleration=300
     maxLinearVelocity=700
     velocityScaleZ=0
     externalLinearDragV=<0 0.5 0>
    }}
  }
 }
 RotOwnAxis {type=boolean}
 Sim Space Settings {
  type=SimSpaceSettings
  default {
   dampingAlpha=1
   maxAngularAcceleration=10000
   maxAngularVelocity=10000
   maxLinearAcceleration=10000
   maxLinearVelocity=10000
   velocityScaleZ=1
   worldAlpha=0
   externalLinearDragV=<0 0 0>
  }
 }
 Skel Mesh {type=SkeletalMeshComponent}
 StartCrouching {type=boolean}
]
