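# Physics asset of collision shapes for the M_MED_CyborgWarrior player body, named as the colliders for its cape.
# Capsules cover pelvis, spine, neck and clavicles, with a head sphere; arm, leg and foot bodies are empty.
# Back skirt and dagger dynamic bones carry sized capsules, while their auto, base and offset bones have tiny ones.
# A large root capsule sits mostly below the origin like a floor, and no constraints exist.
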PhysicsAsset /BRCosmetics/Characters/Player/Male/Medium/Bodies/M_MED_CyborgWarrior/Meshes/M_MED_CyborgWarrior_CapeColliders
bodies {
 auto_skirt_back [{shapeType=Capsule, length=0.505, radius=0.505, shapeName=auto_skirt_back_capsule, rotation=<rot 0 -90 0>}]
 auto_skirt_midback [{shapeType=Capsule, length=0.505, radius=0.505, shapeName=auto_skirt_midback_capsule, rotation=<rot 0 -90 0>}]
 base_daggers_back [{shapeType=Capsule, length=0.505, radius=0.505, shapeName=base_daggers_back_capsule, rotation=<rot 0 -90 0>}]
 base_skirt_back [{shapeType=Capsule, length=0.505, radius=0.505, shapeName=base_skirt_back_capsule, rotation=<rot 0 -90 0>}]
 base_skirt_midback [{shapeType=Capsule, length=0.505, radius=0.505, shapeName=base_skirt_midback_capsule, rotation=<rot 0 -90 0>}]
 calf_l []
 calf_r []
 clavicle_l [{shapeType=Capsule, length=9.0192, radius=9.8871, center=<9.4266 -1.1806 -2.4977>, rotation=<rot 69.3297 1.4476 176.9462>}]
 clavicle_r [{shapeType=Capsule, length=9.0192, radius=9.8871, center=<-9.4266 1.1806 2.4977>, rotation=<rot 69.3297 1.4476 -3.0538>}]
 dyn_daggers_back [{shapeType=Capsule, length=28, radius=6, shapeName=dyn_daggers_back_capsule, center=<0 0 -1.1>, rotation=<rot 0 0 100>}]
 dyn_skirt_back_01 [{shapeType=Capsule, length=9, radius=4.5, shapeName=dyn_skirt_back_01_capsule, center=<4.0414 0 0>, rotation=<rot 90 0 0>}]
 dyn_skirt_back_02 [{shapeType=Capsule, length=16, radius=7.3, shapeName=dyn_skirt_back_02_capsule, center=<9.7972 1.4386 -2>, rotation=<rot 90 -0.0003 0>}]
 dyn_skirt_back_03 [{shapeType=Capsule, length=11, radius=10, shapeName=dyn_skirt_back_03_capsule, center=<11.1078 0 -5>, rotation=<rot 0 0 90>}]
 dyn_skirt_back_04 [{shapeType=Capsule, length=11, radius=10, shapeName=dyn_skirt_back_04_capsule, center=<6.1555 0.2232 -6>, rotation=<rot -0.0003 -0.0001 90>}]
 dyn_skirt_midback_01 [{shapeType=Capsule, length=9, radius=4.5, shapeName=dyn_skirt_midback_01_capsule, center=<4.0414 -0.8324 0>, rotation=<rot 90 0 0>}]
 dyn_skirt_midback_02 [{shapeType=Capsule, length=11, radius=5.5, shapeName=dyn_skirt_midback_02_capsule, center=<7.5592 -0.5697 -2>, rotation=<rot 90 -0.0005 0>}]
 dyn_skirt_midback_03 [{shapeType=Capsule, length=11, radius=9, shapeName=dyn_skirt_midback_03_capsule, center=<11.1078 0 -5>, rotation=<rot 0 0 90>}]
 dyn_skirt_midback_04 [{shapeType=Capsule, length=11, radius=9, shapeName=dyn_skirt_midback_04_capsule, center=<6.8316 -0.0328 -6>, rotation=<rot 0 0 90>}]
 foot_l []
 foot_r []
 head [{shapeType=Sphere, radius=17.3125, center=<11.028 -0.174 0>}]
 lowerarm_l []
 lowerarm_r []
 neck_02 [{shapeType=Capsule, length=13.4375, radius=7.3175, center=<-0.2642 -0.3408 0>, rotation=<rot 90 -171.1193 -175.3978>}]
 offset_skirt_back [{shapeType=Capsule, length=0.505, radius=0.505, shapeName=offset_skirt_back_capsule, rotation=<rot 0 -90 0>}]
 offset_skirt_midback [{shapeType=Capsule, length=0.505, radius=0.505, shapeName=offset_skirt_midback_capsule, rotation=<rot 0 -90 0>}]
 pelvis [{shapeType=Capsule, length=8.6456, radius=16, center=<-2.4887 0.1881 0>}]
 root [{shapeType=Capsule, length=0.5, radius=88, shapeName=root_capsule, center=<0 0 -85.6534>}]
 spine_01 [{shapeType=Capsule, length=7.9438, radius=14, center=<-1.7965 0.1884 0>}]
 spine_02 [
  {shapeType=Capsule, length=7.3264, radius=14, center=<-1.711 -0.6777 0.0001>}
  {shapeType=Capsule, length=29.3956, radius=11.5807, center=<2.5024 0.4992 3.9825>, rotation=<rot 90 0.0049 0.0049>}
  {shapeType=Capsule, length=29.875, radius=11.1013, center=<2.4704 0.4991 -3.9825>, rotation=<rot 90 -0.0396 -0.0396>}
 ]
 spine_03 [{shapeType=Capsule, length=8.2107, radius=14, center=<0.1445 -1.7338 0.0001>}]
 spine_04 [{shapeType=Capsule, length=9.4404, radius=15, center=<2.766 -1.8323 0.0002>}]
 spine_05 [
  {shapeType=Capsule, length=16, radius=16, center=<-2.2795 -1.237 0>}
  {shapeType=Capsule, length=13.9518, radius=13.9425, center=<-0.2916 0.4452 0>}
 ]
 thigh_l []
 thigh_r []
 upperarm_l []
 upperarm_r []
}
constraints []
