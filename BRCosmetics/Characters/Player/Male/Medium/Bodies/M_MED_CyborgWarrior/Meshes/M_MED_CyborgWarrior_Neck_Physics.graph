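# Physics asset for the neck mesh of the male medium Cyborg Warrior player body, defining capsule collision and joints for six dynamic neck bones.
# Small capsules on base bones at the back, mid and front of each side anchor larger capsules on the matching dyn bones.
# Each constraint allows swing mainly on one axis, keeping the other swing and twist nearly rigid, so pieces hinge.
# A head capsule has no constraint.
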PhysicsAsset /BRCosmetics/Characters/Player/Male/Medium/Bodies/M_MED_CyborgWarrior/Meshes/M_MED_CyborgWarrior_Neck_Physics
bodies {
 base_neck_backmid_l [{shapeType=Capsule, length=0.505, radius=0.505, shapeName=base_neck_backmid_l_capsule, rotation=<rot 0 -90 0>}]
 base_neck_backmid_r [{shapeType=Capsule, length=0.505, radius=0.505, shapeName=base_neck_backmid_r_capsule, rotation=<rot 0 -90 0>}]
 base_neck_frontmid_l [{shapeType=Capsule, length=0.505, radius=0.505, shapeName=base_neck_frontmid_l_capsule, rotation=<rot 0 -90 0>}]
 base_neck_frontmid_r [{shapeType=Capsule, length=0.505, radius=0.505, shapeName=base_neck_frontmid_r_capsule, rotation=<rot 0 -90 0>}]
 base_neck_mid_l [{shapeType=Capsule, length=0.505, radius=0.505, shapeName=base_neck_mid_l_capsule, rotation=<rot 0 -90 0>}]
 base_neck_mid_r [{shapeType=Capsule, length=0.505, radius=0.505, shapeName=base_neck_mid_r_capsule, rotation=<rot 0 -90 0>}]
 dyn_neck_backmid_l [{shapeType=Capsule, length=9, radius=4, shapeName=dyn_neck_backmid_l_capsule, center=<5 0 3.5>, rotation=<rot 0 0 90>}]
 dyn_neck_backmid_r [{shapeType=Capsule, length=9, radius=4, shapeName=dyn_neck_backmid_r_capsule, center=<5 0 3.5>, rotation=<rot 0 0 90>}]
 dyn_neck_frontmid_l [{shapeType=Capsule, length=9, radius=4, shapeName=dyn_neck_frontmid_l_capsule, center=<7 0 3.5>, rotation=<rot 0 0 90>}]
 dyn_neck_frontmid_r [{shapeType=Capsule, length=9, radius=4, shapeName=dyn_neck_frontmid_r_capsule, center=<7 0 3.5>, rotation=<rot 0 0 90>}]
 dyn_neck_mid_l [{shapeType=Capsule, length=9, radius=4, shapeName=dyn_neck_mid_l_capsule, center=<5 0 5>, rotation=<rot 0 0 75>}]
 dyn_neck_mid_r [{shapeType=Capsule, length=9, radius=4, shapeName=dyn_neck_mid_r_capsule, center=<5 0 4.5>, rotation=<rot 0 0 90>}]
 head [{shapeType=Capsule, length=9, radius=7.5, shapeName=head_capsule, center=<0 2 0>, rotation=<rot 90 0 0>}]
}
constraints [
 {bone1Name=dyn_neck_backmid_l, bone2Name=base_neck_backmid_l, swing1LimitDegrees=1, swing2LimitDegrees=45, twistLimitDegrees=1}
 {bone1Name=dyn_neck_backmid_r, bone2Name=base_neck_backmid_r, swing1LimitDegrees=1, swing2LimitDegrees=45, twistLimitDegrees=1}
 {bone1Name=dyn_neck_frontmid_l, bone2Name=base_neck_frontmid_l, swing1LimitDegrees=1, swing2LimitDegrees=45, twistLimitDegrees=1}
 {bone1Name=dyn_neck_frontmid_r, bone2Name=base_neck_frontmid_r, swing1LimitDegrees=1, swing2LimitDegrees=45, twistLimitDegrees=1}
 {bone1Name=dyn_neck_mid_l, bone2Name=base_neck_mid_l, swing1LimitDegrees=1, swing2LimitDegrees=45, twistLimitDegrees=1}
 {bone1Name=dyn_neck_mid_r, bone2Name=base_neck_mid_r, swing1LimitDegrees=1, swing2LimitDegrees=45, twistLimitDegrees=1}
]
