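# Control Rig for the Cyborg Warrior's face accessory part that procedurally offsets hood bones as the character's pose changes.
# Head-to-neck Euler deltas are remapped into weights, one through a curve, that drive additive transform corrections on the hood bones.
# Spherical pose readers add two more corrections: left upper arm pose gated by Crouching, and head pose gated by Targeting while not crouching, smoothed with an alpha interp.
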ControlRigBlueprint /BRCosmetics/Characters/Player/Male/Medium/Bodies/M_MED_CyborgWarrior/Meshes/Parts/M_MED_CyborgWarrior_FaceAcc_CtrlRig
graphs [
 ConstructionGraph {nodes [RigUnit_PrepareForExecution = &RigUnit_PrepareForExecution()]}
 Multiply_8_ContainedGraph {nodes [
   n1 = RigVMFunctionEntryNode()
   RigVMFunctionReturnNode(Result=&RigVMFunction_MathDoubleMul(B=@n1.C, A=&RigVMFunction_MathDoubleMul(A=@n1.A, B=@n1.B).Result).Result)
  ]}
 New Function_ContainedGraph {nodes [
   n1 = RigVMFunctionEntryNode()
   RigVMFunction_MathQuaternionToEuler = &RigVMFunction_MathQuaternionToEuler(
    RotationOrder=ZYX
    Value=&RigUnit_GetRelativeTransformForItem(
     Child=(Type=Bone,Name="head")
     Parent=(Type=Bone,Name="neck_01")
     bChildInitial=False
     bParentInitial=False
     Child.Name=@n1.Child
     Parent.Name=@n1.Parent
    ).RelativeTransform.Rotation
   )
   Remap = &RigVMFunction_MathDoubleRemap(
    TargetMaximum=1
    TargetMinimum=0
    SourceMaximum=@n1.Source Max
    SourceMinimum=@n1.Source Min
    Value=@RigVMFunction_MathQuaternionToEuler.Result.Z
   )
   > RigVMFunctionReturnNode(
    Delta_z=@Remap.Result
    Delta_X_Negative=&RigVMFunction_MathDoubleRemap(SourceMaximum=-60, SourceMinimum=-1, TargetMaximum=1, TargetMinimum=0, Value=@RigVMFunction_MathQuaternionToEuler.Result.X).Result
    Delta_X_Positive=&RigVMFunction_MathDoubleRemap(SourceMaximum=60, SourceMinimum=0, TargetMaximum=1, TargetMinimum=0, Value=@RigVMFunction_MathQuaternionToEuler.Result.X).Result
   )
  ]}
 New Function_ContainedGraph {nodes [
   n1 = RigVMFunctionEntryNode()
   RigVMFunction_MathBoolToInteger = &RigVMFunction_MathBoolToInteger(Value=@n1.onlyTwist)
   RigVMFunction_MathQuaternionSwingTwist = &RigVMFunction_MathQuaternionSwingTwist(
    twistAxis=@n1.twistAxis
    Input=&RigUnit_GetRelativeTransformForItem(
     bChildInitial=False
     bParentInitial=False
     child=(Type=Bone,Name="None")
     parent=(Type=Bone,Name="None")
     child.Name=@n1.child
     parent.Name=@n1.parent
     parent.Type=@n1.parentType
    ).RelativeTransform.Rotation
   )
   RigVMFunction_MathQuaternionToEuler = &RigVMFunction_MathQuaternionToEuler(
    Value=RigVMDispatchNode(
     Values=((),())
     Index=@RigVMFunction_MathBoolToInteger.Result
     Values.0=@RigVMFunction_MathQuaternionSwingTwist.Swing
     Values.1=@RigVMFunction_MathQuaternionSwingTwist.Twist
    ) {template="DISPATCH_RigVMDispatch_SelectInt32(in Index,in Values,out Result)"}.Result
    RotationOrder=@n1.Rotation Order
   )
   > DISPATCH_RigVMDispatch_Print = RigVMDispatchNode(ScreenColor=(R=1,G=1,B=1,A=1), ScreenDuration=1.5, Enabled=@n1.debugRotation, Value=@RigVMFunction_MathQuaternionToEuler.Result) {template="DISPATCH_RigVMDispatch_Print(in Prefix,in Value,in Enabled,in ScreenDuration,in ScreenColor)"}
   Clamp = &RigVMFunction_MathIntClamp(Value=@n1.primAxis)
   DISPATCH_RigVMDispatch_SelectInt32 = RigVMDispatchNode(
    Values=(0,0,0)
    Index=@Clamp.Result
    Values.0=@RigVMFunction_MathQuaternionToEuler.Result.X
    Values.1=@RigVMFunction_MathQuaternionToEuler.Result.Y
    Values.2=@RigVMFunction_MathQuaternionToEuler.Result.Z
   ) {template="DISPATCH_RigVMDispatch_SelectInt32(in Index,in Values,out Result)"}
   Remap = &RigVMFunction_MathDoubleRemap(
    TargetMaximum=1
    TargetMinimum=0
    bClamp=False
    Value=@DISPATCH_RigVMDispatch_SelectInt32.Result
    SourceMaximum=@n1.max
    SourceMinimum=@n1.min
   )
   Multiply = &RigVMFunction_MathDoubleMul(A=&RigVMFunction_MathDoubleClamp(Value=@Remap.Result).Result, B=@n1.premult)
   > RigVMFunctionReturnNode(outputWeight=@Multiply.Result)
  ]}
 RigVMFunctionLibrary {nodes [
   DeltaRotation = RigVMCollapseNode(ExecuteContext=())
   DeltaRotation_1 = RigVMCollapseNode(ExecuteContext=(), Rotation Order=ZYX, premult=1, twistAxis=(X=1,Y=0,Z=0))
  ]}
 RigVMFunction_Sequence_1_ContainedGraph {nodes [
   n1 = RigVMFunctionEntryNode()
   RigVMFunction_Sequence = &RigVMFunction_Sequence(ExecuteContext=@n1.ExecuteContext)
   n2 = &RigVMFunction_Sequence(ExecuteContext=@RigVMFunction_Sequence.B)
   n3 = &RigVMFunction_Sequence(ExecuteContext=@n2.B)
   RigVMFunctionReturnNode(A=@RigVMFunction_Sequence.A, B=@n2.A, C=@n3.A, D=@n3.B)
  ]}
 RigVMModel {nodes [
   n1 = RigVMRerouteNode(Value=())
   RigUnit_GetTransform = &RigUnit_GetTransform(Item=(Type=Control,Name="head_ctrl"), bInitial=False)
   RigUnit_SetTransform = &RigUnit_SetTransform(Item=(Type=Bone,Name="head"), Weight=1, bInitial=False, ExecutePin=@n1.Value, Value=@RigUnit_GetTransform.Transform)
   RigUnit_BeginExecution = &RigUnit_BeginExecution()
   RigVMFunction_Sequence = RigVMAggregateNode(ExecuteContext=@RigUnit_BeginExecution.ExecutePin)
   DeltaRotation_2 = RigVMFunctionReferenceNode(Child=head, Parent=neck_01, Source Max=-60, Source Min=-20, ExecuteContext=@RigVMFunction_Sequence.B)
   RigVMFunction_AnimEvalRichCurve = &RigVMFunction_AnimEvalRichCurve(
    Curve="(EditorCurveData=(Keys=((InterpMode=RCIM_Linear,TangentMode=RCTM_Auto,TangentWeightMode=RCTWM_WeightedNone,Time=0,Value=0,ArriveTangent=0,ArriveTangentWeight=0,LeaveTangent=0,LeaveTangentWeight=0),(InterpMode=RCIM_Linear,TangentMode=RCTM_Auto,TangentWeight ...(truncated)"
    SourceMinimum=0
    TargetMinimum=0
    Value=@DeltaRotation_2.Delta_X_Negative
   )
   > &RigUnit_ModifyTransforms(
    ItemToModify="((Item=(Type=Bone,Name=\"base_hood_l1\"),Transform=(Rotation=(X=0,Y=0,Z=0,W=-1),Translation=(X=4,Y=0,Z=0),Scale3D=(X=1,Y=1,Z=0.5))),(Item=(Type=Bone,Name=\"base_hood_l\"),Transform=(Rotation=(X=0,Y=0,Z=0,W=1),Translation=(X=22,Y=0,Z=0),Scale3D=(X=1,Y=1,Z=1))), ...(truncated)"
    WeightMinimum=0
    Weight=@RigVMFunction_AnimEvalRichCurve.Result
   )
   > &RigUnit_ModifyTransforms(
    ItemToModify="((Item=(Type=Bone,Name=\"base_hood_r1\"),Transform=(Rotation=(X=0,Y=0,Z=0,W=1),Translation=(X=-11,Y=-14.5,Z=4),Scale3D=(X=1,Y=1,Z=1))),(Item=(Type=Bone,Name=\"base_hood_r\"),Transform=(Rotation=(X=0,Y=0,Z=0,W=1),Translation=(X=-22,Y=-20,Z=-5),Scale3D=(X=1,Y=1, ...(truncated)"
    WeightMinimum=0
    Weight=@DeltaRotation_2.Delta_X_Positive
   )
   > &RigUnit_ModifyTransforms(
    ItemToModify=((Item=(Type=Bone,Name="None"),Transform=(Rotation=(X=0,Y=0,Z=0,W=1),Translation=(X=0,Y=0,Z=0),Scale3D=(X=1,Y=1,Z=1))),(Item=(Type=Bone,Name="dyn_hood_c"),Transform=(Rotation=(X=0,Y=0,Z=0,W=1),Translation=(X=0,Y=0,Z=0),Scale3D=(X=1,Y=1,Z=1))))
    WeightMinimum=0
    Weight=@DeltaRotation_2.Delta_z
   )
   > DeltaRotation_3 = RigVMFunctionReferenceNode(Child=head, Parent=neck_01, Source Max=-60, Source Min=-20)
   > &RigUnit_ModifyTransforms(
    ItemToModify="((Item=(Type=Bone,Name=\"base_hood_l1\"),Transform=(Rotation=(X=0,Y=0,Z=0,W=1),Translation=(X=-0.2555,Y=-7.3279,Z=0),Scale3D=(X=1,Y=1,Z=1))),(Item=(Type=Bone,Name=\"base_hood_l\"),Transform=(Rotation=(X=0,Y=0,Z=0,W=1),Translation=(X=1.5745,Y=0,Z=0),Scale3D=(X= ...(truncated)"
    WeightMinimum=0
    Weight=@DeltaRotation_3.Delta_z
   )
   RigVMFunction_MathBoolMake = /Script/RigVM.RigVMFunction_MathBoolMake(Value=false) {is=struct RigVMUnitNode}
   n2 = &RigUnit_SphericalPoseReader(
    ActiveRegionScaleFactors=(PositiveWidth=1,NegativeWidth=1,PositiveHeight=1,NegativeHeight=1)
    ActiveRegionSize=0.1
    Debug=(bDrawDebug=false,bDraw2D=False,bDrawLocalAxes=False,DebugScale=25,DebugSegments=20,DebugThickness=0.25)
    DriverAxis=(X=1,Y=0,Z=0)
    DriverItem=(Type=Bone,Name="upperarm_l")
    FalloffRegionScaleFactors=(PositiveWidth=1,NegativeWidth=1,PositiveHeight=1,NegativeHeight=1)
    FalloffSize=0.2
    FlipHeightScaling=False
    FlipWidthScaling=false
    OptionalParentItem=(Type=Bone,Name="None")
    RotationOffset=(X=-10,Y=-15,Z=0)
    Debug.bDrawDebug=@RigVMFunction_MathBoolMake.Value
    ExecutePin=@RigVMFunction_Sequence.C
   )
   RigVMFunction_MathBoolToFloat = &RigVMFunction_MathBoolToFloat(Value=RigVMVariableNode() {variable=Crouching}.Value)
   Multiply = &RigVMFunction_MathDoubleMul(B=@RigVMFunction_MathBoolToFloat.Result, A=@n2.OutputParam)
   > &RigUnit_ModifyTransforms(
    ItemToModify="((Item=(Type=Bone,Name=\"base_hood_l\"),Transform=(Rotation=(X=-0.1305,Y=0,Z=0,W=0.9914),Translation=(X=-15,Y=0,Z=0),Scale3D=(X=1,Y=1,Z=1))),(Item=(Type=Bone,Name=\"base_hood_l1\"),Transform=(Rotation=(X=0,Y=0,Z=0,W=1),Translation=(X=-13,Y=0,Z=0),Scale3D=(X=1, ...(truncated)"
    WeightMinimum=0
    Weight=@Multiply.Result
   )
   > n3 = &RigUnit_SphericalPoseReader(
    ActiveRegionScaleFactors=(PositiveWidth=1,NegativeWidth=0.5,PositiveHeight=0.75,NegativeHeight=0.75)
    ActiveRegionSize=0.1
    Debug=(bDrawDebug=false,bDraw2D=False,bDrawLocalAxes=False,DebugScale=25,DebugSegments=20,DebugThickness=0.25)
    DriverAxis=(X=0,Y=1,Z=0)
    DriverItem=(Type=Bone,Name="head")
    FalloffRegionScaleFactors=(PositiveWidth=1,NegativeWidth=0.5,PositiveHeight=0.5,NegativeHeight=0.5)
    FalloffSize=0.2
    FlipHeightScaling=false
    FlipWidthScaling=true
    OptionalParentItem=(Type=Bone,Name="None")
    RotationOffset=(X=100,Y=0,Z=0)
    Debug.bDrawDebug=@RigVMFunction_MathBoolMake.Value
   )
   RigVMFunction_MathBoolNot = &RigVMFunction_MathBoolNot(Value=RigVMVariableNode() {variable=Crouching}.Value)
   Multiply_3 = RigVMAggregateNode(
    B=&RigVMFunction_MathBoolToFloat(Value=RigVMVariableNode() {variable=Targeting}.Value).Result
    C=&RigVMFunction_MathBoolToFloat(Value=@RigVMFunction_MathBoolNot.Result).Result
    A=@n3.OutputParam
   )
   AlphaInterp = &RigVMFunction_AlphaInterp(InterpSpeedDecreasing=12, InterpSpeedIncreasing=12, Scale=1, bClampResult=true, bMapRange=true, Value=@Multiply_3.Result)
   > &RigUnit_ModifyTransforms(
    ItemToModify=((Item=(Type=Bone,Name="base_hood_r"),Transform=(Rotation=(X=0,Y=0,Z=0,W=1),Translation=(X=-6,Y=0,Z=6),Scale3D=(X=1,Y=1,Z=1))),(Item=(Type=Bone,Name="base_hood_r1"),Transform=(Rotation=(X=0,Y=0,Z=0,W=1),Translation=(X=0,Y=0,Z=0),Scale3D=(X=1,Y=1,Z=1))))
    WeightMinimum=0
    Weight=@AlphaInterp.Result
   )
  ]}
]
hierarchy [
 root {loc=<0 0 0>, rot=<0 0 0 1>}
 pelvis {parent=root, loc=<0 -0.7071 93.1191>, rot=<0.0127 0.707 -0.0127 -0.707>}
 spine_01 {parent=pelvis, loc=<6.7494 -0.6388 0>, rot=<0 0 0.034 -0.9994>}
 spine_02 {parent=spine_01, loc=<7.2518 -0.035 0>, rot=<0 0 0.0238 -0.9997>}
 spine_03 {parent=spine_02, loc=<6.0104 0 0>, rot=<0 0 0.0149 -0.9999>}
 spine_04 {parent=spine_03, loc=<8.1024 0 0>, rot=<0 0 -0.0132 -0.9999>}
 spine_05 {parent=spine_04, loc=<10.0146 0 0>, rot=<0 0 -0.0015 -1>}
 neck_01 {parent=spine_05, loc=<11.3138 0.1529 0>, rot=<0 0 -0.2003 -0.9797>}
 neck_02 {parent=neck_01, loc=<4.5627 0 0>, rot=<0 0 0.0504 -0.9987>}
 head {parent=neck_02, loc=<5.3914 0 0>, rot=<0 0 0.115 -0.9934>}
 base_hood_l {parent=neck_02, loc=<7.2212 -8.7365 -1>, rot=<-0.0784 -0.7027 0.0784 -0.7027>}
 dyn_hood_l {parent=base_hood_l, loc=<0 0 0>, rot=<0 0 0 1>}
 base_hood_c {parent=neck_02, loc=<8.5829 -8.8852 0>, rot=<-0.0784 -0.7027 0.0784 -0.7027>}
 dyn_hood_c {parent=base_hood_c, loc=<0 0 0>, rot=<0 0 0 1>}
 base_hood_back_top {parent=dyn_hood_c, loc=<0.9529 -0.3646 2.3154>, rot=<-0.2756 -0.0943 0.4855 -0.8243>}
 dyn_hood_back_top {parent=base_hood_back_top, loc=<0 0 0>, rot=<0 0 0 1>}
 base_hood_r {parent=neck_02, loc=<7.2212 -8.7365 1>, rot=<-0.0784 -0.7027 0.0784 -0.7027>}
 dyn_hood_r {parent=base_hood_r, loc=<0 0 0>, rot=<0 0 0 1>}
 base_hood_r1 {parent=neck_02, loc=<2 7.5 11>, rot=<-0.0784 -0.7027 0.0784 -0.7027>}
 dyn_hood_r1 {parent=base_hood_r1, loc=<0 0 0>, rot=<0 0 0 1>}
 base_hood_l1 {parent=neck_02, loc=<2 7.5 -11>, rot=<-0.0784 -0.7027 0.0784 -0.7027>}
 dyn_hood_l1 {parent=base_hood_l1, loc=<0 0 0>, rot=<0 0 0 1>}
 base_hood_off_01a {parent=neck_01, loc=<0.6686 -8.2004 -0.3979>, rot=<0.1335 -0.7016 0.6853 -0.1429>}
 dyn_hood_off_01a {parent=base_hood_off_01a, loc=<0 0 0>, rot=<0 0 0 1>}
 dyn_hood_off_02 {parent=dyn_hood_off_01a, loc=<4.674 0 0>, rot=<0 -0.151 0 -0.9885>}
 base_hood_off_01b {parent=neck_01, loc=<3.4905 -8.4848 -0.4625>, rot=<0.4566 -0.5399 0.5399 -0.4566>}
 dyn_hood_off_01b {parent=base_hood_off_01b, loc=<0 0 0>, rot=<0 0 0 1>}
 clavicle_l {parent=spine_05, loc=<4.3575 -0.1466 -2.3223>, rot=<0.0465 -0.7289 -0.009 -0.6829>}
 upperarm_l {parent=clavicle_l, loc=<15.5111 0 0>, rot=<0.0226 -0.3234 -0.0088 -0.9459>}
 upperarm_twist_01_l {parent=upperarm_l, loc=<0.0041 0 0>, rot=<0 0 0 -1>}
 clavicle_r {parent=spine_05, loc=<4.3575 -0.1466 2.3223>, rot=<0.7289 0.0465 -0.6829 0.009>}
 upperarm_r {parent=clavicle_r, loc=<-15.5111 0 0>, rot=<0.0226 -0.3234 -0.0088 -0.9459>}
 upperarm_twist_01_r {parent=upperarm_r, loc=<-0.0041 0 0>, rot=<0 0 0 -1>}
 head_null {parent=spine_05, type=Null, loc=<20.6573 3.5489 0>, rot=<0 0 -0.036 -0.9994>}
 head_ctrl {controlType=EULER_TRANSFORM, type=Control, loc=<0 0 0>, rot=<0 0 0 1>}
]
variables [
 Crouching {type=bool, default=False}
 Targeting {type=bool, default=False}
]
